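# Physics asset for the DBNO (downed) state of the PageTruffle medium male body, covering its lower-body cloth bones.
# Capsules cover pelvis, thighs, calves and dyn_Cloth_volume chains: A, B and C per side, plus a short back chain.
# Chains hang from the pelvis, with back, B and C roots angularly locked.
# Cross-constraints tie neighboring chains at lower segments (A-B, B-C), so each side moves as a connected sheet.
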PhysicsAsset /BRCosmetics/Characters/Player/Male/Medium/Bodies/M_MED_PageTruffle/Meshes/M_MED_PageTruffle_DBNO_Physics
bodies {
 calf_l [{
   shapeType=Capsule
   length=36.2787
   radius=7.359
   shapeName=calf_l_capsule
   center=<-28.7966 0 -0.6293>
   rotation=<rot 88.7481 179.9996 179.9997>
  }]
 calf_r [{
   shapeType=Capsule
   length=35.1434
   radius=7.359
   shapeName=calf_r_capsule
   center=<28.6369 0 0.6258>
   rotation=<rot 88.7481 179.9996 179.9997>
  }]
 dyn_Cloth_volume_bk [{
   shapeType=Capsule
   length=7
   radius=4.5
   shapeName=dyn_Cloth_volume_bk_capsule
   center=<-4.7395 1.8107 0.2328>
   rotation=<rot -87.1875 179.9999 -180>
  }]
 dyn_Cloth_volume_bk01 [{
   shapeType=Capsule
   length=7
   radius=4.5
   shapeName=dyn_Cloth_volume_bk01_capsule
   center=<-4.7395 -1.6471 0.2328>
   rotation=<rot 2.8125 179.9997 180>
  }]
 dyn_Cloth_volume_l_A [{
   shapeType=Capsule
   length=14.3794
   radius=8.0445
   shapeName=dyn_Cloth_volume_l_A_capsule
   center=<-4.7395 -4.2584 0.2328>
   rotation=<rot -2.1454 0.0002 0>
  }]
 dyn_Cloth_volume_l_A01 [{
   shapeType=Capsule
   length=14.3794
   radius=8.0445
   shapeName=dyn_Cloth_volume_l_A01_capsule
   center=<-4.7395 -6.4845 0.2328>
   rotation=<rot -2.1454 0.0002 0>
  }]
 dyn_Cloth_volume_l_A02 [{
   shapeType=Capsule
   length=1.7019
   radius=9.649
   shapeName=dyn_Cloth_volume_l_C03_capsule
   center=<-5.9707 -6.963 0.355>
   rotation=<rot 4.6471 -179.9999 -180>
  }]
 dyn_Cloth_volume_l_A03 [{
   shapeType=Capsule
   length=3.35
   radius=8.0009
   shapeName=dyn_Cloth_volume_l_C03_capsule
   center=<-5.9707 -4.6676 0.355>
   rotation=<rot 4.6471 -179.9999 -180>
  }]
 dyn_Cloth_volume_l_A04 [{
   shapeType=Capsule
   length=7.2989
   radius=8.8634
   shapeName=dyn_Cloth_volume_l_C03_capsule
   center=<-7.6412 -3.0667 0.0453>
   rotation=<rot -4.0896 -179.9997 -180>
  }]
 dyn_Cloth_volume_l_B [{
   shapeType=Capsule
   length=7
   radius=4.5
   shapeName=dyn_Cloth_volume_l_B_capsule
   center=<-4.7395 1.8107 0.2328>
   rotation=<rot -87.1876 -179.9998 179.9998>
  }]
 dyn_Cloth_volume_l_B01 [{
   shapeType=Capsule
   length=7
   radius=4.5
   shapeName=dyn_Cloth_volume_l_B01_capsule
   center=<-4.7395 3.2451 0.2328>
   rotation=<rot -87.1876 -179.9998 179.9998>
  }]
 dyn_Cloth_volume_l_B02 [{
   shapeType=Capsule
   length=13.926
   radius=9.5239
   shapeName=dyn_Cloth_volume_r_B03_capsule
   center=<-4.7395 -8.0251 0.2328>
   rotation=<rot 3.9271 180 -180>
  }]
 dyn_Cloth_volume_l_B03 [{
   shapeType=Capsule
   length=15.5741
   radius=7.8758
   shapeName=dyn_Cloth_volume_r_B03_capsule
   center=<-4.7395 -6.5603 0.2328>
   rotation=<rot 3.9271 180 -180>
  }]
 dyn_Cloth_volume_l_B04 [{
   shapeType=Capsule
   length=24.5149
   radius=7.8758
   shapeName=dyn_Cloth_volume_r_B03_capsule
   center=<-4.7395 -5.7226 0.2328>
   rotation=<rot 3.9271 180 -180>
  }]
 dyn_Cloth_volume_l_C [{
   shapeType=Capsule
   length=7
   radius=4.5
   shapeName=dyn_Cloth_volume_l_C_capsule
   center=<-4.7395 1.8107 0.2328>
   rotation=<rot -87.1875 -180 179.9999>
  }]
 dyn_Cloth_volume_l_C01 [{
   shapeType=Capsule
   length=4.9977
   radius=6.5023
   shapeName=dyn_Cloth_volume_l_C01_capsule
   center=<-4.7396 -3.0096 0.2328>
   rotation=<rot 2.8124 179.9998 180>
  }]
 dyn_Cloth_volume_l_C02 [{
   shapeType=Capsule
   length=8.1898
   radius=6.7957
   shapeName=dyn_Cloth_volume_l_C02_capsule
   center=<-3.8912 -9.9071 0.3239>
   rotation=<rot -2.8124 -0.0003 0.9373>
  }]
 dyn_Cloth_volume_l_C03 [{
   shapeType=Capsule
   length=13.5541
   radius=7.9314
   shapeName=dyn_Cloth_volume_l_C03_capsule
   center=<-4.0715 -9.1819 0.2006>
   rotation=<rot -2.8125 -0.0002 0>
  }]
 dyn_Cloth_volume_l_C04 [{
   shapeType=Capsule
   length=10.5838
   radius=10.394
   shapeName=dyn_Cloth_volume_l_C03_capsule
   center=<-6.7104 -4.9639 0.3302>
   rotation=<rot -2.8125 -0.0002 0>
  }]
 dyn_Cloth_volume_r_A [{
   shapeType=Capsule
   length=14.3794
   radius=8.0445
   shapeName=dyn_Cloth_volume_r_A_capsule
   center=<-4.7395 -4.2584 0.2328>
   rotation=<rot -2.1456 0.0002 0>
  }]
 dyn_Cloth_volume_r_A01 [{
   shapeType=Capsule
   length=14.3794
   radius=8.0445
   shapeName=dyn_Cloth_volume_r_A01_capsule
   center=<-4.7395 -6.4845 0.2328>
   rotation=<rot -2.1456 0.0002 0>
  }]
 dyn_Cloth_volume_r_A02 [{
   shapeType=Capsule
   length=3.4751
   radius=9.1868
   shapeName=dyn_Cloth_volume_l_C03_capsule
   center=<-5.9606 -10.075 0.4798>
   rotation=<rot 4.6471 -179.9999 175.3171>
  }]
 dyn_Cloth_volume_r_A03 [{
   shapeType=Capsule
   length=3.9205
   radius=8.7414
   shapeName=dyn_Cloth_volume_l_C03_capsule
   center=<-5.9545 -9.2485 0.5541>
   rotation=<rot 4.6471 -179.9999 173.9541>
  }]
 dyn_Cloth_volume_r_A04 [{
   shapeType=Capsule
   length=1.1812
   radius=8.4034
   shapeName=dyn_Cloth_volume_l_C03_capsule
   center=<-7.6694 -5.7501 0.4399>
   rotation=<rot 3.8051 -179.9998 -179.9999>
  }]
 dyn_Cloth_volume_r_B [{
   shapeType=Capsule
   length=7
   radius=4.5
   shapeName=dyn_Cloth_volume_r_B_capsule
   center=<-4.7395 1.8107 0.2328>
   rotation=<rot -87.1874 179.9998 -179.9998>
  }]
 dyn_Cloth_volume_r_B01 [{
   shapeType=Capsule
   length=7
   radius=4.5
   shapeName=dyn_Cloth_volume_r_B01_capsule
   center=<-4.7395 1.8107 0.2328>
   rotation=<rot -87.1874 179.9998 -179.9998>
  }]
 dyn_Cloth_volume_r_B02 [{
   shapeType=Capsule
   length=13.926
   radius=9.5239
   shapeName=dyn_Cloth_volume_r_B03_capsule
   center=<-4.7395 -8.0251 0.2328>
   rotation=<rot 3.9271 180 -180>
  }]
 dyn_Cloth_volume_r_B03 [{
   shapeType=Capsule
   length=15.5741
   radius=7.8758
   shapeName=dyn_Cloth_volume_r_B03_capsule
   center=<-4.7395 -6.5603 0.2328>
   rotation=<rot 3.9271 180 -180>
  }]
 dyn_Cloth_volume_r_B04 [{
   shapeType=Capsule
   length=21.0041
   radius=7.8758
   shapeName=dyn_Cloth_volume_r_B03_capsule
   center=<-4.7395 -5.7226 0.2328>
   rotation=<rot 3.9271 180 -180>
  }]
 dyn_Cloth_volume_r_C [{
   shapeType=Capsule
   length=7
   radius=4.5
   shapeName=dyn_Cloth_volume_r_C_capsule
   center=<-4.7395 1.8107 0.2328>
   rotation=<rot -87.1875 179.9998 -179.9999>
  }]
 dyn_Cloth_volume_r_C01 [{
   shapeType=Capsule
   length=4.9977
   radius=6.5023
   shapeName=dyn_Cloth_volume_r_C01_capsule
   center=<-4.7396 -3.0096 0.2328>
   rotation=<rot 2.8126 179.9998 -180>
  }]
 dyn_Cloth_volume_r_C02 [{
   shapeType=Capsule
   length=8.1898
   radius=6.7957
   shapeName=dyn_Cloth_volume_l_C02_capsule
   center=<-3.8912 -9.9071 0.3239>
   rotation=<rot -2.8124 -0.0003 0.9373>
  }]
 dyn_Cloth_volume_r_C03 [{
   shapeType=Capsule
   length=13.5541
   radius=7.9314
   shapeName=dyn_Cloth_volume_l_C03_capsule
   center=<-4.0715 -9.1819 0.2006>
   rotation=<rot -2.8125 -0.0002 0>
  }]
 dyn_Cloth_volume_r_C04 [{
   shapeType=Capsule
   length=10.5838
   radius=10.394
   shapeName=dyn_Cloth_volume_l_C03_capsule
   center=<-6.7104 -4.9639 0.3302>
   rotation=<rot -2.8125 -0.0002 0>
  }]
 pelvis [{shapeType=Capsule, length=9.4921, radius=11.9478, center=<1.9277 -0.3348 0.4765>, rotation=<rot -3.4833 -90.0002 177.7238>}]
 thigh_l [{shapeType=Capsule, length=29.4468, radius=7.415, center=<-23.0185 -1.4423 1.2568>, rotation=<rot 0.0376 -86.592 -90.72>}]
 thigh_r [{shapeType=Capsule, length=29.5305, radius=7.2647, center=<23.0122 1.7952 -1.0598>, rotation=<rot -0.0699 -85.064 -88.8823>}]
}
constraints [
 {bone1Name=thigh_l, bone2Name=pelvis, swing1LimitDegrees=45, swing2LimitDegrees=45, twistLimitDegrees=45}
 {bone1Name=thigh_r, bone2Name=pelvis, swing1LimitDegrees=45, swing2LimitDegrees=45, twistLimitDegrees=45}
 {bone1Name=dyn_Cloth_volume_l_A03, bone2Name=dyn_Cloth_volume_l_A02, swing1LimitDegrees=45, swing2LimitDegrees=45, twistLimitDegrees=15}
 {bone1Name=dyn_Cloth_volume_l_A02, bone2Name=dyn_Cloth_volume_l_A01, swing1LimitDegrees=45, swing2LimitDegrees=45, twistLimitDegrees=15}
 {bone1Name=dyn_Cloth_volume_l_B03, bone2Name=dyn_Cloth_volume_l_B02, swing1LimitDegrees=45, swing2LimitDegrees=15, twistLimitDegrees=5}
 {bone1Name=dyn_Cloth_volume_l_B02, bone2Name=dyn_Cloth_volume_l_B01, swing1LimitDegrees=45, swing2LimitDegrees=15, twistLimitDegrees=15}
 {
  bone1Name=dyn_Cloth_volume_l_B01
  bone2Name=dyn_Cloth_volume_l_B
  swing1LimitDegrees=45
  swing1Motion=Locked
  swing2LimitDegrees=45
  swing2Motion=Locked
  twistLimitDegrees=45
  twistMotion=Locked
 }
 {bone1Name=dyn_Cloth_volume_l_C03, bone2Name=dyn_Cloth_volume_l_C02, swing1LimitDegrees=45, swing2LimitDegrees=5, twistLimitDegrees=5}
 {bone1Name=dyn_Cloth_volume_l_C02, bone2Name=dyn_Cloth_volume_l_C01, swing1LimitDegrees=45, swing2LimitDegrees=15, twistLimitDegrees=15}
 {bone1Name=dyn_Cloth_volume_l_C01, bone2Name=dyn_Cloth_volume_l_C, swing1LimitDegrees=15, swing2LimitDegrees=5, twistLimitDegrees=5}
 {bone1Name=dyn_Cloth_volume_r_A02, bone2Name=dyn_Cloth_volume_r_A01, swing1LimitDegrees=45, swing2LimitDegrees=45, twistLimitDegrees=15}
 {bone1Name=dyn_Cloth_volume_r_A03, bone2Name=dyn_Cloth_volume_r_A02, swing1LimitDegrees=45, swing2LimitDegrees=45, twistLimitDegrees=15}
 {bone1Name=dyn_Cloth_volume_r_A01, bone2Name=dyn_Cloth_volume_r_A, swing1LimitDegrees=45, swing2LimitDegrees=45, twistLimitDegrees=45}
 {bone1Name=dyn_Cloth_volume_l_A01, bone2Name=dyn_Cloth_volume_l_A, swing1LimitDegrees=45, swing2LimitDegrees=45, twistLimitDegrees=45}
 {bone1Name=dyn_Cloth_volume_bk01, bone2Name=dyn_Cloth_volume_bk, swing1LimitDegrees=15, swing2LimitDegrees=5, twistLimitDegrees=5}
 {
  bone1Name=dyn_Cloth_volume_bk
  bone2Name=pelvis
  swing1LimitDegrees=45
  swing1Motion=Locked
  swing2LimitDegrees=45
  swing2Motion=Locked
  twistLimitDegrees=45
  twistMotion=Locked
 }
 {bone1Name=dyn_Cloth_volume_l_A, bone2Name=pelvis, swing1LimitDegrees=45, swing2LimitDegrees=45, twistLimitDegrees=45}
 {
  bone1Name=dyn_Cloth_volume_l_B
  bone2Name=pelvis
  swing1LimitDegrees=45
  swing1Motion=Locked
  swing2LimitDegrees=45
  swing2Motion=Locked
  twistLimitDegrees=45
  twistMotion=Locked
 }
 {
  bone1Name=dyn_Cloth_volume_l_C
  bone2Name=pelvis
  swing1LimitDegrees=45
  swing1Motion=Locked
  swing2LimitDegrees=45
  swing2Motion=Locked
  twistLimitDegrees=45
  twistMotion=Locked
 }
 {bone1Name=dyn_Cloth_volume_r_A, bone2Name=pelvis, swing1LimitDegrees=45, swing2LimitDegrees=45, twistLimitDegrees=45}
 {bone1Name=dyn_Cloth_volume_r_C03, bone2Name=dyn_Cloth_volume_r_C02, swing1LimitDegrees=45, swing2LimitDegrees=5, twistLimitDegrees=5}
 {bone1Name=dyn_Cloth_volume_r_C02, bone2Name=dyn_Cloth_volume_r_C01, swing1LimitDegrees=45, swing2LimitDegrees=15, twistLimitDegrees=15}
 {bone1Name=dyn_Cloth_volume_r_C01, bone2Name=dyn_Cloth_volume_r_C, swing1LimitDegrees=15, swing2LimitDegrees=5, twistLimitDegrees=5}
 {
  bone1Name=dyn_Cloth_volume_r_C
  bone2Name=pelvis
  swing1LimitDegrees=45
  swing1Motion=Locked
  swing2LimitDegrees=45
  swing2Motion=Locked
  twistLimitDegrees=45
  twistMotion=Locked
 }
 {bone1Name=dyn_Cloth_volume_r_B03, bone2Name=dyn_Cloth_volume_r_B02, swing1LimitDegrees=45, swing2LimitDegrees=15, twistLimitDegrees=15}
 {bone1Name=dyn_Cloth_volume_r_B02, bone2Name=dyn_Cloth_volume_r_B01, swing1LimitDegrees=45, swing2LimitDegrees=15, twistLimitDegrees=15}
 {
  bone1Name=dyn_Cloth_volume_r_B01
  bone2Name=dyn_Cloth_volume_r_B
  swing1LimitDegrees=45
  swing1Motion=Locked
  swing2LimitDegrees=45
  swing2Motion=Locked
  twistLimitDegrees=45
  twistMotion=Locked
 }
 {
  bone1Name=dyn_Cloth_volume_r_B
  bone2Name=pelvis
  swing1LimitDegrees=45
  swing1Motion=Locked
  swing2LimitDegrees=45
  swing2Motion=Locked
  twistLimitDegrees=45
  twistMotion=Locked
 }
 {bone1Name=dyn_Cloth_volume_r_C04, bone2Name=dyn_Cloth_volume_r_C03, swing1LimitDegrees=45, swing2LimitDegrees=5, twistLimitDegrees=5}
 {bone1Name=dyn_Cloth_volume_r_B04, bone2Name=dyn_Cloth_volume_r_B03, swing1LimitDegrees=45, swing2LimitDegrees=45, twistLimitDegrees=15}
 {bone1Name=dyn_Cloth_volume_l_C04, bone2Name=dyn_Cloth_volume_l_C03, swing1LimitDegrees=45, swing2LimitDegrees=5, twistLimitDegrees=5}
 {bone1Name=dyn_Cloth_volume_l_B04, bone2Name=dyn_Cloth_volume_l_B03, swing1LimitDegrees=45, swing2LimitDegrees=15, twistLimitDegrees=5}
 {bone1Name=dyn_Cloth_volume_l_A04, bone2Name=dyn_Cloth_volume_l_A03, swing1LimitDegrees=45, swing2LimitDegrees=45, twistLimitDegrees=15}
 {bone1Name=dyn_Cloth_volume_r_A04, bone2Name=dyn_Cloth_volume_r_A03, swing1LimitDegrees=45, swing2LimitDegrees=45, twistLimitDegrees=15}
 {
  bone1Name=dyn_Cloth_volume_l_A03
  bone2Name=dyn_Cloth_volume_l_B03
  swing1LimitDegrees=45
  swing1Motion=Locked
  swing2LimitDegrees=45
  twistLimitDegrees=45
 }
 {
  bone1Name=dyn_Cloth_volume_l_A02
  bone2Name=dyn_Cloth_volume_l_B02
  swing1LimitDegrees=45
  swing1Motion=Locked
  swing2LimitDegrees=45
  twistLimitDegrees=45
 }
 {
  bone1Name=dyn_Cloth_volume_r_A03
  bone2Name=dyn_Cloth_volume_r_B03
  swing1LimitDegrees=45
  swing1Motion=Locked
  swing2LimitDegrees=45
  twistLimitDegrees=45
 }
 {
  bone1Name=dyn_Cloth_volume_r_A02
  bone2Name=dyn_Cloth_volume_r_B02
  swing1LimitDegrees=45
  swing1Motion=Locked
  swing2LimitDegrees=45
  twistLimitDegrees=45
 }
 {bone1Name=dyn_Cloth_volume_r_B02, bone2Name=dyn_Cloth_volume_r_C02, swing1LimitDegrees=15, swing2LimitDegrees=45, twistLimitDegrees=45}
 {bone1Name=dyn_Cloth_volume_r_B03, bone2Name=dyn_Cloth_volume_r_C03, swing1LimitDegrees=15, swing2LimitDegrees=45, twistLimitDegrees=45}
 {bone1Name=dyn_Cloth_volume_l_B03, bone2Name=dyn_Cloth_volume_l_C03, swing1LimitDegrees=45, swing2LimitDegrees=45, twistLimitDegrees=45}
 {bone1Name=dyn_Cloth_volume_l_B02, bone2Name=dyn_Cloth_volume_l_C02, swing1LimitDegrees=45, swing2LimitDegrees=45, twistLimitDegrees=45}
 {
  bone1Name=dyn_Cloth_volume_r_B04
  bone2Name=dyn_Cloth_volume_r_C04
  swing1LimitDegrees=45
  swing1Motion=Free
  swing2LimitDegrees=45
  swing2Motion=Free
  twistLimitDegrees=45
  twistMotion=Free
 }
 {bone1Name=dyn_Cloth_volume_l_B04, bone2Name=dyn_Cloth_volume_l_C04, swing1LimitDegrees=45, swing2LimitDegrees=45, twistLimitDegrees=45}
 {
  bone1Name=calf_r
  bone2Name=thigh_r
  swing1LimitDegrees=45
  swing1Motion=Free
  swing2LimitDegrees=45
  swing2Motion=Free
  twistLimitDegrees=45
  twistMotion=Free
 }
 {
  bone1Name=calf_l
  bone2Name=thigh_l
  swing1LimitDegrees=45
  swing1Motion=Free
  swing2LimitDegrees=45
  swing2Motion=Free
  twistLimitDegrees=45
  twistMotion=Free
 }
]
