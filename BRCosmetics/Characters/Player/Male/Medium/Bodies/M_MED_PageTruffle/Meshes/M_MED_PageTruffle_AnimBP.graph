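# Cosmetic Anim Blueprint for the M_MED_PageTruffle body that animates its skirt cloth bones over the attached parent mesh's pose.
# A skirt Control Rig and crouch bone offsets feed a cached pose that drives an Anim Dynamics chain and constraint-based shaping against pelvis and legs.
# RigidBody passes use separate gameplay, front-end and DBNO physics assets.
# Bool and alpha blends select results by movement, crouch, diving, DBNO and front-end state.
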AnimBlueprint /BRCosmetics/Characters/Player/Male/Medium/Bodies/M_MED_PageTruffle/Meshes/M_MED_PageTruffle_AnimBP
target_skeleton=&/M_MED_PageTruffle_Skeleton
graphs [
 AnimGraph {nodes [
   n1 = "Constraint - dyn_Cloth_volume_l_C01 (PP)"!(
    ComponentPose=Local To Component(LocalPose=Use cached pose 'SkirtCachePose'())
    Alpha=1
    ConstraintWeights_0=0.85
    ConstraintWeights_1:float=0.15
   ) -> (Pose) {
    alphaBoolBlend {blendOption=Linear}
    boneToModify {boneName=dyn_Cloth_volume_l_C01}
    constraintSetup [{targetBone {boneName=pelvis}}, {targetBone {boneName=dyn_Cloth_volume_l_B}}]
    constraintWeights [1,1]
   }
   n2 = "Constraint - dyn_Cloth_volume_r_C01 (PP)"!(ComponentPose=@n1, Alpha=1, ConstraintWeights_0=0.85, ConstraintWeights_1:float=0.15) -> (Pose) {
    alphaBoolBlend {blendOption=Linear}
    boneToModify {boneName=dyn_Cloth_volume_r_C01}
    constraintSetup [{targetBone {boneName=pelvis}}, {targetBone {boneName=dyn_Cloth_volume_r_B}}]
    constraintWeights [1,1]
   }
   n3 = "Constraint - dyn_Cloth_volume_l_C03 (PP)"!(ComponentPose=@n2, Alpha=1, ConstraintWeights_0=1, ConstraintWeights_1:float=0) -> (Pose) {
    alphaBoolBlend {blendOption=Linear}
    boneToModify {boneName=dyn_Cloth_volume_l_C03}
    constraintSetup [{targetBone {boneName=dyn_Cloth_volume_l_C02}}, {targetBone {boneName=dyn_Cloth_volume_r_C04}}]
    constraintWeights [1,1]
   }
   n4 = "Constraint - dyn_Cloth_volume_r_C03 (PP)"!(ComponentPose=@n3, Alpha=1, ConstraintWeights_0=1, ConstraintWeights_1:float=0) -> (Pose) {
    alphaBoolBlend {blendOption=Linear}
    boneToModify {boneName=dyn_Cloth_volume_r_C03}
    constraintSetup [{targetBone {boneName=dyn_Cloth_volume_r_C02}}, {targetBone {boneName=dyn_Cloth_volume_r_C04}}]
    constraintWeights [1,1]
   }
   n5 = "Transform (Modify) Bone - Bone: dyn_Cloth_volume_l_C01"(ComponentPose=@n4, Alpha=1, Translation="0,0,0", Rotation="0,20,0", Scale="1,1,1") {
    rotationMode=BMM_Additive
    rotationSpace=BCS_BoneSpace
    scaleMode=BMM_Additive
    scaleSpace=BCS_BoneSpace
    translationMode=BMM_Additive
    translationSpace=BCS_BoneSpace
    alphaBoolBlend {blendOption=Linear}
    boneToModify {boneName=dyn_Cloth_volume_l_C01}
    rotation {yaw=20}
   }
   n6 = "Transform (Modify) Bone - Bone: dyn_Cloth_volume_r_C01"(ComponentPose=@n5, Alpha=1, Translation="0,0,0", Rotation="0,20,0", Scale="1,1,1") {
    rotationMode=BMM_Additive
    rotationSpace=BCS_BoneSpace
    scaleMode=BMM_Additive
    scaleSpace=BCS_BoneSpace
    translationMode=BMM_Additive
    translationSpace=BCS_BoneSpace
    alphaBoolBlend {blendOption=Linear}
    boneToModify {boneName=dyn_Cloth_volume_r_C01}
    rotation {yaw=20}
   }
   n7 = "Transform (Modify) Bone - Bone: dyn_Cloth_volume_l_C03"(ComponentPose=@n6, Alpha=1, Translation="0,0,0", Rotation="0,10,0", Scale="1,1,1") {
    rotationMode=BMM_Additive
    rotationSpace=BCS_BoneSpace
    scaleMode=BMM_Additive
    scaleSpace=BCS_BoneSpace
    translationMode=BMM_Additive
    translationSpace=BCS_BoneSpace
    alphaBoolBlend {blendOption=Linear}
    boneToModify {boneName=dyn_Cloth_volume_l_C03}
    rotation {yaw=10}
   }
   n8 = "Transform (Modify) Bone - Bone: dyn_Cloth_volume_r_C03"(ComponentPose=@n7, Alpha=1, Translation="0,0,0", Rotation="0,10,0", Scale="1,1,1") {
    rotationMode=BMM_Additive
    rotationSpace=BCS_BoneSpace
    scaleMode=BMM_Additive
    scaleSpace=BCS_BoneSpace
    translationMode=BMM_Additive
    translationSpace=BCS_BoneSpace
    alphaBoolBlend {blendOption=Linear}
    boneToModify {boneName=dyn_Cloth_volume_r_C03}
    rotation {yaw=10}
   }
   n9 = "Constraint - dyn_Cloth_volume_bk (PP)"!(ComponentPose=@n8, Alpha=1, ConstraintWeights_0=0.5, ConstraintWeights_1:float=0.5) -> (Pose) {
    alphaBoolBlend {blendOption=Linear}
    boneToModify {boneName=dyn_Cloth_volume_bk}
    constraintSetup [{targetBone {boneName=dyn_Cloth_volume_l_C01}}, {targetBone {boneName=dyn_Cloth_volume_r_C01}}]
    constraintWeights [1,1]
   }
   n10 = "Constraint - dyn_Cloth_volume_bk02 (PP)"!(ComponentPose=@n9, Alpha=1, ConstraintWeights_0=0.5, ConstraintWeights_1:float=0.5) -> (Pose) {
    alphaBoolBlend {blendOption=Linear}
    boneToModify {boneName=dyn_Cloth_volume_bk02}
    constraintSetup [{targetBone {boneName=dyn_Cloth_volume_l_C03}}, {targetBone {boneName=dyn_Cloth_volume_r_C03}}]
    constraintWeights [1,1]
   }
   n11 = Component To Local(ComponentPose=@n10)
   n12 = "Transform (Modify) Bone - Bone: dyn_Cloth_volume_l_C01"(ComponentPose=Local To Component(LocalPose=Use cached pose 'MovingPose'()), Alpha=1, Translation="0,0,0", Rotation="0,0,0", Scale="1,1,1") {
    rotationMode=BMM_Additive
    rotationSpace=BCS_BoneSpace
    scaleMode=BMM_Additive
    scaleSpace=BCS_BoneSpace
    translationMode=BMM_Additive
    translationSpace=BCS_BoneSpace
    alphaBoolBlend {blendOption=Linear}
    boneToModify {boneName=dyn_Cloth_volume_l_C01}
   }
   n13 = "Transform (Modify) Bone - Bone: dyn_Cloth_volume_r_C01"(ComponentPose=@n12, Alpha=1, Translation="0,0,0", Rotation="0,15,0", Scale="1,1,1") {
    rotationMode=BMM_Additive
    rotationSpace=BCS_BoneSpace
    scaleMode=BMM_Additive
    scaleSpace=BCS_BoneSpace
    translationMode=BMM_Additive
    translationSpace=BCS_BoneSpace
    alphaBoolBlend {blendOption=Linear}
    boneToModify {boneName=dyn_Cloth_volume_r_C01}
    rotation {yaw=15}
   }
   n14 = "Transform (Modify) Bone - Bone: dyn_Cloth_volume_l_C04"(ComponentPose=@n13, Alpha=1, Translation="0,0,0", Rotation="0,-10,0", Scale="1,1,1") {
    rotationMode=BMM_Additive
    rotationSpace=BCS_BoneSpace
    scaleMode=BMM_Additive
    scaleSpace=BCS_BoneSpace
    translationMode=BMM_Additive
    translationSpace=BCS_BoneSpace
    alphaBoolBlend {blendOption=Linear}
    boneToModify {boneName=dyn_Cloth_volume_l_C04}
    rotation {yaw=-10}
   }
   n15 = "Transform (Modify) Bone - Bone: dyn_Cloth_volume_r_C04"(ComponentPose=@n14, Alpha=1, Translation="0,0,0", Rotation="0,-30,0", Scale="1,1,1") {
    rotationMode=BMM_Additive
    rotationSpace=BCS_BoneSpace
    scaleMode=BMM_Additive
    scaleSpace=BCS_BoneSpace
    translationMode=BMM_Additive
    translationSpace=BCS_BoneSpace
    alphaBoolBlend {blendOption=Linear}
    boneToModify {boneName=dyn_Cloth_volume_r_C04}
    rotation {yaw=-30}
   }
   n16 = "Constraint - dyn_Cloth_volume_l_C01 (PP)"(ComponentPose=@n15, Alpha=1, ConstraintWeights_0=0.1, ConstraintWeights_1=0.9) {
    alphaBoolBlend {blendOption=Linear}
    boneToModify {boneName=dyn_Cloth_volume_l_C01}
    constraintSetup [{targetBone {boneName=thigh_l}}, {targetBone {boneName=dyn_Cloth_volume_l_C}}]
    constraintWeights [1,1]
   }
   n17 = "Constraint - dyn_Cloth_volume_r_C01 (PP)"(ComponentPose=@n16, Alpha=1, ConstraintWeights_0=0.1, ConstraintWeights_1=0.9) {
    alphaBoolBlend {blendOption=Linear}
    boneToModify {boneName=dyn_Cloth_volume_r_C01}
    constraintSetup [{targetBone {boneName=thigh_r}}, {targetBone {boneName=dyn_Cloth_volume_r_C}}]
    constraintWeights [1,1]
   }
   n18 = "Constraint - dyn_Cloth_volume_l_C03 (PP)"(ComponentPose=@n17, Alpha=1, ConstraintWeights_0=0.5, ConstraintWeights_1=0.5) {
    alphaBoolBlend {blendOption=Linear}
    boneToModify {boneName=dyn_Cloth_volume_l_C03}
    constraintSetup [{targetBone {boneName=calf_l}}, {targetBone {boneName=dyn_Cloth_volume_l_C03}}]
    constraintWeights [1,1]
   }
   n19 = "Constraint - dyn_Cloth_volume_r_C03 (PP)"(ComponentPose=@n18, Alpha=1, ConstraintWeights_0=0.5, ConstraintWeights_1=0.5) {
    alphaBoolBlend {blendOption=Linear}
    boneToModify {boneName=dyn_Cloth_volume_r_C03}
    constraintSetup [{targetBone {boneName=calf_r}}, {targetBone {boneName=dyn_Cloth_volume_r_C03}}]
    constraintWeights [1,1]
   }
   n20 = "Constraint - dyn_Cloth_volume_bk01 (PP)"!(ComponentPose=@n19, Alpha=1, ConstraintWeights_0=0.5, ConstraintWeights_1:float=0.5) -> (Pose) {
    alphaBoolBlend {blendOption=Linear}
    boneToModify {boneName=dyn_Cloth_volume_bk01}
    constraintSetup [{targetBone {boneName=dyn_Cloth_volume_l_C01}}, {targetBone {boneName=dyn_Cloth_volume_r_C01}}]
    constraintWeights [1,1]
   }
   n21 = Component To Local(ComponentPose=@n20)
   n22 = Get MovingAnd NotCrouchNotDivingNotEmote() -> (MovingAnd NotCrouchNotDivingNotEmote:bool)
   n23 = "Constraint - dyn_Cloth_volume_r_C02 (RR)"!(
    ComponentPose=Local To Component(LocalPose=Use cached pose 'SkirtCachePose'())
    Alpha=1
    ConstraintWeights_0=0.2
    ConstraintWeights_1:float=0.8
   ) -> (Pose) {
    alphaBoolBlend {blendOption=Linear}
    boneToModify {boneName=dyn_Cloth_volume_r_C02}
    constraintSetup [{transformType=Rotation, targetBone {boneName=thigh_r}}, {transformType=Rotation, targetBone {boneName=dyn_Cloth_volume_r_C02}}]
    constraintWeights [1,1]
   }
   n24 = "Constraint - dyn_Cloth_volume_l_C02 (RR)"!(ComponentPose=@n23, Alpha=1, ConstraintWeights_0=0.2, ConstraintWeights_1:float=0.8) -> (Pose) {
    alphaBoolBlend {blendOption=Linear}
    boneToModify {boneName=dyn_Cloth_volume_l_C02}
    constraintSetup [{transformType=Rotation, targetBone {boneName=thigh_l}}, {transformType=Rotation, targetBone {boneName=dyn_Cloth_volume_l_C02}}]
    constraintWeights [1,1]
   }
   n25 = "Constraint - dyn_Cloth_volume_l_B02 (PP)"!(ComponentPose=@n24, Alpha=1, ConstraintWeights_0=0.5, ConstraintWeights_1:float=0.5) -> (Pose) {
    alphaBoolBlend {blendOption=Linear}
    boneToModify {boneName=dyn_Cloth_volume_l_B02}
    constraintSetup [{targetBone {boneName=dyn_Cloth_volume_l_C02}}, {targetBone {boneName=dyn_Cloth_volume_l_A02}}]
    constraintWeights [1,1]
   }
   n26 = "Constraint - dyn_Cloth_volume_r_B02 (PP)"!(ComponentPose=@n25, Alpha=1, ConstraintWeights_0=0.5, ConstraintWeights_1:float=0.5) -> (Pose) {
    alphaBoolBlend {blendOption=Linear}
    boneToModify {boneName=dyn_Cloth_volume_r_B02}
    constraintSetup [{targetBone {boneName=dyn_Cloth_volume_r_C02}}, {targetBone {boneName=dyn_Cloth_volume_r_A02}}]
    constraintWeights [1,1]
   }
   n27 = RigidBody!(ComponentPose=@n26, bAlphaBoolEnabled:bool, OverrideWorldGravity:Vector="0,0,-980", ExternalForce) -> (Pose) {
    alphaInputType=Bool
    bClampLinearTranslationLimitToRefPose=true
    bOverrideWorldGravity=true
    overridePhysicsAsset=&/M_MED_PageTruffle_DBNO_Physics
    simulationSpace=BaseBoneSpace
    alphaBoolBlend {blendOption=Linear}
    baseBoneRef {boneName=gravity}
    componentAppliedLinearAccClamp=<3000 3000 2700>
    componentLinearAccScale=<0.01 0.01 0.01>
    componentLinearVelScale=<0.08 0.4 0.4>
    overrideWorldGravity {z=-980}
   }
   n28 = Two Way Blend(A=Use cached pose 'DBNOPose'(), B=Component To Local(ComponentPose=@n27), Alpha=Get DBNO_Alpha() -> (DBNO_Alpha:double)) {blendNode {alphaBoolBlend {blendOption=Linear}}}
   n29 = Use cached pose 'SkirtDynPose'()
   n30 = Get DivingOrDBNO() -> (DivingOrDBNO:bool)
   Output Pose(Result=Blend Poses by bool(
     BlendPose_0=Blend Poses by bool(BlendPose_0=@n11, BlendPose_1=@n21, BlendTime_0=0.2, BlendTime_1=0.2, bActiveValue=@n22) {blendType=Linear}
     BlendPose_1=Blend Poses by bool(BlendPose_0=@n28, BlendPose_1=@n29, bActiveValue=@n30) {blendType=Linear}
    ) {blendType=Linear}) {layerGroup=None}
   n31 = "Constraint - thigh_twist_01_l_fix (PP)"!(
    ComponentPose=Local To Component(LocalPose=Use cached pose 'FrontEndPose'())
    Alpha=1
    ConstraintWeights_0=0.8
    ConstraintWeights_1:float=0.2
   ) -> (Pose) {
    alphaBoolBlend {blendOption=Linear}
    boneToModify {boneName=thigh_twist_01_l_fix}
    constraintSetup [{targetBone {boneName=pelvis}}, {targetBone {boneName=thigh_l}}]
    constraintWeights [1,1]
   }
   n32 = "Constraint - thigh_twist_01_r_fix (PP)"!(ComponentPose=@n31, Alpha=1, ConstraintWeights_0=0.8, ConstraintWeights_1:float=0.2) -> (Pose) {
    alphaBoolBlend {blendOption=Linear}
    boneToModify {boneName=thigh_twist_01_r_fix}
    constraintSetup [{targetBone {boneName=pelvis}}, {targetBone {boneName=thigh_r}}]
    constraintWeights [1,1]
   }
   n33 = "Transform (Modify) Bone - Bone: thigh_twist_01_l_fix"(ComponentPose=@n32, Alpha=1, Translation="0,0,0", Rotation="5,5,-5", Scale="1,1,1") {
    rotationMode=BMM_Additive
    rotationSpace=BCS_BoneSpace
    scaleMode=BMM_Additive
    scaleSpace=BCS_BoneSpace
    translationMode=BMM_Additive
    translationSpace=BCS_BoneSpace
    alphaBoolBlend {blendOption=Linear}
    boneToModify {boneName=thigh_twist_01_l_fix}
    rotation=<rot 5 5 -5>
   }
   n34 = "Transform (Modify) Bone - Bone: thigh_twist_01_r_fix"(ComponentPose=@n33, Alpha=1, Translation="0,0,0", Rotation="2,5,0", Scale="1,1,1") {
    rotationMode=BMM_Additive
    rotationSpace=BCS_BoneSpace
    scaleMode=BMM_Additive
    scaleSpace=BCS_BoneSpace
    translationMode=BMM_Additive
    translationSpace=BCS_BoneSpace
    alphaBoolBlend {blendOption=Linear}
    boneToModify {boneName=thigh_twist_01_r_fix}
    rotation {pitch=2, yaw=5}
   }
   n35 = "Constraint - deform_knee_in_l_ALT (PP)"!(ComponentPose=@n34, Alpha=1, ConstraintWeights_0=0.5, ConstraintWeights_1:float=0.5) -> (Pose) {
    alphaBoolBlend {blendOption=Linear}
    boneToModify {boneName=deform_knee_in_l_ALT}
    constraintSetup [{targetBone {boneName=pelvis}}, {targetBone {boneName=thigh_twist_01_l_fix}}]
    constraintWeights [1,1]
   }
   n36 = "Constraint - deform_knee_in_r_ALT (PP)"!(ComponentPose=@n35, Alpha=1, ConstraintWeights_0=0.5, ConstraintWeights_1:float=0.5) -> (Pose) {
    alphaBoolBlend {blendOption=Linear}
    boneToModify {boneName=deform_knee_in_r_ALT}
    constraintSetup [{targetBone {boneName=pelvis}}, {targetBone {boneName=thigh_twist_01_r_fix}}]
    constraintWeights [1,1]
   }
   n37 = Component To Local(ComponentPose=@n36)
   n38 = Use cached pose 'FrontEndPose'()
   n39 = Get isFrontEndAndNotEmote() -> (isFrontEndAndNotEmote:bool)
   n40 = RigidBody(
    ComponentPose=Local To Component(LocalPose=Use cached pose 'MovingPose'())
    bAlphaBoolEnabled=Get is_WearingCape() -> (is_WearingCape:bool)
    OverrideWorldGravity="0,0,-980"
   ) {
    alphaInputType=Bool
    bClampLinearTranslationLimitToRefPose=true
    bOverrideWorldGravity=true
    overridePhysicsAsset=&/M_MED_PageTruffle_Physics
    simulationSpace=BaseBoneSpace
    alphaBoolBlend {blendOption=Linear}
    baseBoneRef {boneName=gravity}
    componentAppliedLinearAccClamp=<3000 3000 2700>
    componentLinearAccScale=<0.01 0.4 0.35>
    componentLinearVelScale=<0.08 0.4 0.4>
    overrideWorldGravity {z=-980}
   }
   n41 = Component To Local(ComponentPose=@n40)
   n42 = RigidBody(
    ComponentPose=Local To Component(LocalPose=Use cached pose 'MovingPose'())
    bAlphaBoolEnabled=Get is_WearingCape()
    OverrideWorldGravity="0,0,-980"
   ) {
    alphaInputType=Bool
    bClampLinearTranslationLimitToRefPose=true
    bOverrideWorldGravity=true
    overridePhysicsAsset=&M_MED_PageTruffle_Physics
    simulationSpace=BaseBoneSpace
    alphaBoolBlend {blendOption=Linear}
    baseBoneRef {boneName=gravity}
    componentAppliedLinearAccClamp=<3000 3000 2700>
    componentLinearAccScale=<0.01 0.4 0.35>
    componentLinearVelScale=<0.08 0.4 0.4>
    overrideWorldGravity {z=-980}
   }
   n43 = Component To Local(ComponentPose=@n42)
   n44 = Get MovingAnd NotCrouchNotDivingNotEmote()
   Save cached pose 'SkirtDynPose'(Pose=Blend Poses by bool(
     BlendPose_0=Blend Poses by bool(BlendPose_0=@n37, BlendPose_1=@n38, bActiveValue=@n39) {blendType=Linear}
     BlendPose_1=Blend Poses by bool(BlendPose_0=@n43, BlendPose_1=@n41, BlendTime_1=0.5, bActiveValue=@n44) {blendType=Linear}
     BlendTime_1=0.5
     bActiveValue=Get bIsInFrontEnd()
    ) {blendType=Linear}) {cacheName=SkirtDynPose}
   n45 = Get thigh_shape_fn_r_0() -> (thigh_shape_fn_r_0:double)
   n46 = "Transform (Modify) Bone - Bone: dyn_Cloth_volume_l_C"(
    ComponentPose=Local To Component(LocalPose=Use cached pose 'CrouchingPose'())
    Alpha=@n45
    Translation="0,0,0"
    Rotation="0,32,0"
    Scale="1,1,1"
   ) {
    rotationMode=BMM_Additive
    rotationSpace=BCS_BoneSpace
    scaleMode=BMM_Additive
    scaleSpace=BCS_BoneSpace
    translationMode=BMM_Additive
    translationSpace=BCS_BoneSpace
    alphaBoolBlend {blendOption=Linear}
    boneToModify {boneName=dyn_Cloth_volume_l_C}
    rotation {yaw=32}
   }
   n47 = "Transform (Modify) Bone - Bone: dyn_Cloth_volume_l_C03"(ComponentPose=@n46, Alpha=@n45, Translation="0,0,0", Rotation="0,27,0", Scale="1,1,1") {
    rotationMode=BMM_Additive
    rotationSpace=BCS_BoneSpace
    scaleMode=BMM_Additive
    scaleSpace=BCS_BoneSpace
    translationMode=BMM_Additive
    translationSpace=BCS_BoneSpace
    alphaBoolBlend {blendOption=Linear}
    boneToModify {boneName=dyn_Cloth_volume_l_C03}
    rotation {yaw=27}
   }
   n48 = "Transform (Modify) Bone - Bone: dyn_Cloth_volume_l_B02"(ComponentPose=@n47, Alpha=@n45, Translation="0,0,0", Rotation="0,10,0", Scale="1,1,1") {
    rotationMode=BMM_Additive
    rotationSpace=BCS_BoneSpace
    scaleMode=BMM_Additive
    scaleSpace=BCS_BoneSpace
    translationMode=BMM_Additive
    translationSpace=BCS_BoneSpace
    alphaBoolBlend {blendOption=Linear}
    boneToModify {boneName=dyn_Cloth_volume_l_B02}
    rotation {yaw=10}
   }
   n49 = "Transform (Modify) Bone - Bone: dyn_Cloth_volume_r_C02"(ComponentPose=@n48, Alpha=1, Translation="0,0,0", Rotation="0,20,0", Scale="1,1,1") {
    rotationMode=BMM_Additive
    rotationSpace=BCS_BoneSpace
    scaleMode=BMM_Additive
    scaleSpace=BCS_BoneSpace
    translationMode=BMM_Additive
    translationSpace=BCS_BoneSpace
    alphaBoolBlend {blendOption=Linear}
    boneToModify {boneName=dyn_Cloth_volume_r_C02}
    rotation {yaw=20}
   }
   n50 = "Transform (Modify) Bone - Bone: dyn_Cloth_volume_r_C03"(ComponentPose=@n49, Alpha=1, Translation="0,0,0", Rotation="0,18,0", Scale="1,1,1") {
    rotationMode=BMM_Additive
    rotationSpace=BCS_BoneSpace
    scaleMode=BMM_Additive
    scaleSpace=BCS_BoneSpace
    translationMode=BMM_Additive
    translationSpace=BCS_BoneSpace
    alphaBoolBlend {blendOption=Linear}
    boneToModify {boneName=dyn_Cloth_volume_r_C03}
    rotation {yaw=18}
   }
   n51 = "Transform (Modify) Bone - Bone: dyn_Cloth_volume_r_C04"(ComponentPose=@n50, Alpha=1, Translation="0,0,0", Rotation="0,10,0", Scale="1,1,1") {
    rotationMode=BMM_Additive
    rotationSpace=BCS_BoneSpace
    scaleMode=BMM_Additive
    scaleSpace=BCS_BoneSpace
    translationMode=BMM_Additive
    translationSpace=BCS_BoneSpace
    alphaBoolBlend {blendOption=Linear}
    boneToModify {boneName=dyn_Cloth_volume_r_C04}
    rotation {yaw=10}
   }
   n52 = "Transform (Modify) Bone - Bone: dyn_Cloth_volume_r_B01"(ComponentPose=@n51, Alpha=1, Translation="0,0,0", Rotation="0,18,0", Scale="1,1,1") {
    rotationMode=BMM_Additive
    rotationSpace=BCS_BoneSpace
    scaleMode=BMM_Additive
    scaleSpace=BCS_BoneSpace
    translationMode=BMM_Additive
    translationSpace=BCS_BoneSpace
    alphaBoolBlend {blendOption=Linear}
    boneToModify {boneName=dyn_Cloth_volume_r_B01}
    rotation {yaw=18}
   }
   n53 = Component To Local(ComponentPose=@n52)
   n54 = Use cached pose 'CrouchingPose'()
   n55 = Get CrouchingAndNotMovingAndNotShape() -> (CrouchingAndNotMovingAndNotShape:bool)
   n56 = "Constraint - dyn_Cloth_volume_bk (PP)"(
    ComponentPose=Local To Component(LocalPose=Blend Poses by bool(BlendPose_0=@n53, BlendPose_1=@n54, bActiveValue=@n55) {blendType=Linear})
    Alpha=1
    ConstraintWeights_0=0.5
    ConstraintWeights_1=0.5
   ) {
    alphaBoolBlend {blendOption=Linear}
    boneToModify {boneName=dyn_Cloth_volume_bk}
    constraintSetup [{targetBone {boneName=dyn_Cloth_volume_l_C}}, {targetBone {boneName=dyn_Cloth_volume_r_C}}]
    constraintWeights [1,1]
   }
   n57 = "Constraint - dyn_Cloth_volume_bk01 (PP)"(ComponentPose=@n56, Alpha=1, ConstraintWeights_0=0.5, ConstraintWeights_1=0.5) {
    alphaBoolBlend {blendOption=Linear}
    boneToModify {boneName=dyn_Cloth_volume_bk01}
    constraintSetup [{targetBone {boneName=dyn_Cloth_volume_l_C01}}, {targetBone {boneName=dyn_Cloth_volume_r_C01}}]
    constraintWeights [1,1]
   }
   n58 = "Constraint - dyn_Cloth_volume_bk02 (PP)"(ComponentPose=@n57, Alpha=1, ConstraintWeights_0=0.5, ConstraintWeights_1=0.5) {
    alphaBoolBlend {blendOption=Linear}
    boneToModify {boneName=dyn_Cloth_volume_bk02}
    constraintSetup [{targetBone {boneName=dyn_Cloth_volume_l_C02}}, {targetBone {boneName=dyn_Cloth_volume_r_C02}}]
    constraintWeights [1,1]
   }
   Save cached pose 'SkirtCachePose'(Pose=Component To Local(ComponentPose=@n58)) {cacheName=SkirtCachePose}
   n59 = "Constraint - gravity (TT)"!(
    ComponentPose=Local To Component(LocalPose=Copy Pose From Mesh() {bUseAttachedParent=true})
    Alpha=1
    ConstraintWeights_0=0.3
    ConstraintWeights_1:float=0.7
   ) -> (Pose) {
    alphaBoolBlend {blendOption=Linear}
    boneToModify {boneName=gravity}
    constraintSetup [{transformType=Translation, targetBone {boneName=root}}, {transformType=Translation, targetBone {boneName=pelvis}}]
    constraintWeights [1,1]
   }
   n60 = "Constraint - gravity_cloth (PP)"!(ComponentPose=@n59, Alpha=1, ConstraintWeights_0=1, ConstraintWeights_1:float=0) -> (Pose) {
    alphaBoolBlend {blendOption=Linear}
    boneToModify {boneName=gravity_cloth}
    constraintSetup [{targetBone {boneName=root}}, {targetBone {boneName=pelvis}}]
    constraintWeights [1,1]
   }
   n61 = Control Rig!(Source=Component To Local(ComponentPose=@n60), Alpha, is sliding:double=0) -> (Pose) {
    alphaBoolBlend {blendOption=Linear}
    controlRigAssetReference {blueprintRigClass=&/M_MED_PageTruffle_Skirt_CtrlRig.M_MED_PageTruffle_Skirt_CtrlRig_C}
   }
   Save cached pose 'CrouchingPose'(Pose=@n61) {cacheName=CrouchingPose}
   n62 = RigidBody(
    ComponentPose=Local To Component(LocalPose=Use cached pose 'MovingPose'())
    bAlphaBoolEnabled=Get is_WearingCape()
    OverrideWorldGravity="0,0,-980"
   ) {
    alphaInputType=Bool
    bClampLinearTranslationLimitToRefPose=true
    bOverrideWorldGravity=true
    overridePhysicsAsset=&/M_MED_PageTruffle_FrontendPhysics
    simulationSpace=BaseBoneSpace
    alphaBoolBlend {blendOption=Linear}
    baseBoneRef {boneName=gravity}
    componentAppliedLinearAccClamp=<3000 3000 2700>
    componentLinearAccScale=<0.01 0.4 0.35>
    componentLinearVelScale=<0.08 0.4 0.4>
    overrideWorldGravity {z=-980}
   }
   Save cached pose 'FrontEndPose'(Pose=Component To Local(ComponentPose=@n62)) {cacheName=FrontEndPose}
   n63 = Anim Dynamics - Bone: dyn_Cloth_volume_l_A02!(
    ComponentPose=Local To Component(LocalPose=Use cached pose 'SkirtCachePose'())
    Alpha=1
    GravityOverride:Vector="0,0,500"
    ExternalForce="0,0,0"
   ) -> (Pose) {
    angularDampingOverride=0.99
    angularSpringConstant=800
    bAngularSpring=true
    bOverrideAngularDamping=true
    bUseGravityOverride=true
    gravityScale=0.3
    simulationSpace=BoneRelative
    alphaBoolBlend {blendOption=Linear}
    boundBone {boneName=dyn_Cloth_volume_l_A02}
    componentAppliedLinearAccClamp=<2000 2000 1600>
    componentLinearAccScale=<0.4 0.4 0.4>
    componentLinearVelScale=<0.45 0.45 0.55>
    gravityOverride {z=500}
    physicsBodyDefinitions [{
      collisionType=CustomSphere
      sphereCollisionRadius=2
      boundBone {boneName=dyn_Cloth_volume_l_A02}
      boxExtents=<33 33 33>
      constraintSetup {twistAxis=AxisY, angularLimitsMax=<0 0 31>, angularLimitsMin=<0 0 0>, angularTarget=<0 0 0>}
      localJointOffset=<-5 0.2 -0.1>
     }]
    relativeSpaceBone {boneName=root}
   }
   n64 = Anim Dynamics - Bone: dyn_Cloth_volume_l_A03!(ComponentPose=@n63, Alpha=1, GravityOverride:Vector="0,0,500", ExternalForce="0,0,0") -> (Pose) {
    angularDampingOverride=0.99
    angularSpringConstant=800
    bAngularSpring=true
    bOverrideAngularDamping=true
    bUseGravityOverride=true
    gravityScale=0.3
    simulationSpace=BoneRelative
    alphaBoolBlend {blendOption=Linear}
    boundBone {boneName=dyn_Cloth_volume_l_A03}
    componentAppliedLinearAccClamp=<2000 2000 1600>
    componentLinearAccScale=<0.4 0.4 0.4>
    componentLinearVelScale=<0.45 0.45 0.55>
    gravityOverride {z=500}
    physicsBodyDefinitions [{
      collisionType=CustomSphere
      sphereCollisionRadius=2
      boundBone {boneName=dyn_Cloth_volume_l_A03}
      boxExtents=<27 27 27>
      constraintSetup {twistAxis=AxisY, angularLimitsMax=<0 0 31>, angularLimitsMin=<0 0 0>, angularTarget=<0 0 0>}
      localJointOffset=<-5 0.2 -0.1>
     }]
    relativeSpaceBone {boneName=root}
   }
   n65 = Anim Dynamics - Bone: dyn_Cloth_volume_l_A04!(ComponentPose=@n64, Alpha=1, GravityOverride:Vector="0,0,500", ExternalForce="0,0,0") -> (Pose) {
    angularDampingOverride=0.99
    angularSpringConstant=800
    bAngularSpring=true
    bOverrideAngularDamping=true
    bUseGravityOverride=true
    gravityScale=0.3
    simulationSpace=BoneRelative
    alphaBoolBlend {blendOption=Linear}
    boundBone {boneName=dyn_Cloth_volume_l_A04}
    componentAppliedLinearAccClamp=<2000 2000 1600>
    componentLinearAccScale=<0.4 0.4 0.4>
    componentLinearVelScale=<0.45 0.45 0.55>
    gravityOverride {z=500}
    physicsBodyDefinitions [{
      collisionType=CustomSphere
      sphereCollisionRadius=2
      boundBone {boneName=dyn_Cloth_volume_l_A04}
      boxExtents=<27 27 27>
      constraintSetup {twistAxis=AxisY, angularLimitsMax=<0 0 31>, angularLimitsMin=<0 0 0>, angularTarget=<0 0 0>}
      localJointOffset=<-5 0.2 -0.1>
     }]
    relativeSpaceBone {boneName=root}
   }
   n66 = Anim Dynamics - Bone: dyn_Cloth_volume_l_B!(ComponentPose=@n65, Alpha=1, GravityOverride:Vector="0,0,500", ExternalForce="0,0,0") -> (Pose) {
    angularDampingOverride=0.99
    angularSpringConstant=800
    bAngularSpring=true
    bOverrideAngularDamping=true
    bUseGravityOverride=true
    gravityScale=0.3
    simulationSpace=BoneRelative
    alphaBoolBlend {blendOption=Linear}
    boundBone {boneName=dyn_Cloth_volume_l_B}
    componentAppliedLinearAccClamp=<2000 2000 1600>
    componentLinearAccScale=<0.4 0.4 0.4>
    componentLinearVelScale=<0.45 0.45 0.55>
    gravityOverride {z=500}
    physicsBodyDefinitions [{
      collisionType=CustomSphere
      sphereCollisionRadius=2
      boundBone {boneName=dyn_Cloth_volume_l_B}
      boxExtents=<50 50 50>
      constraintSetup {angularLimitsMax=<10 1 31>, angularLimitsMin=<-10 -1 0>, angularTarget=<0 0 0>}
      localJointOffset=<-5 0.2 -0.1>
     }]
    relativeSpaceBone {boneName=root}
   }
   n67 = Anim Dynamics - Bone: dyn_Cloth_volume_l_B01!(ComponentPose=@n66, Alpha=1, GravityOverride:Vector="0,0,500", ExternalForce="0,0,0") -> (Pose) {
    angularDampingOverride=0.99
    angularSpringConstant=800
    bAngularSpring=true
    bOverrideAngularDamping=true
    bUseGravityOverride=true
    gravityScale=0.3
    simulationSpace=BoneRelative
    alphaBoolBlend {blendOption=Linear}
    boundBone {boneName=dyn_Cloth_volume_l_B01}
    componentAppliedLinearAccClamp=<2000 2000 1600>
    componentLinearAccScale=<0.4 0.4 0.4>
    componentLinearVelScale=<0.45 0.45 0.55>
    gravityOverride {z=500}
    physicsBodyDefinitions [{
      collisionType=CustomSphere
      sphereCollisionRadius=2
      boundBone {boneName=dyn_Cloth_volume_l_B01}
      boxExtents=<33 33 33>
      constraintSetup {angularLimitsMax=<10 1 31>, angularLimitsMin=<-10 -1 0>, angularTarget=<0 0 0>}
      localJointOffset=<-5 0.2 -0.1>
     }]
    relativeSpaceBone {boneName=root}
   }
   n68 = Anim Dynamics - Bone: dyn_Cloth_volume_l_B02!(ComponentPose=@n67, Alpha=1, GravityOverride:Vector="0,0,500", ExternalForce="0,0,0") -> (Pose) {
    angularDampingOverride=0.99
    angularSpringConstant=800
    bAngularSpring=true
    bOverrideAngularDamping=true
    bUseGravityOverride=true
    gravityScale=0.3
    simulationSpace=BoneRelative
    alphaBoolBlend {blendOption=Linear}
    boundBone {boneName=dyn_Cloth_volume_l_B02}
    componentAppliedLinearAccClamp=<2000 2000 1600>
    componentLinearAccScale=<0.4 0.4 0.4>
    componentLinearVelScale=<0.45 0.45 0.55>
    gravityOverride {z=500}
    physicsBodyDefinitions [{
      collisionType=CustomSphere
      sphereCollisionRadius=2
      boundBone {boneName=dyn_Cloth_volume_l_B02}
      boxExtents=<27 27 27>
      constraintSetup {angularLimitsMax=<10 1 31>, angularLimitsMin=<-10 -1 0>, angularTarget=<0 0 0>}
      localJointOffset=<-5 0.2 -0.1>
     }]
    relativeSpaceBone {boneName=root}
   }
   n69 = Anim Dynamics - Bone: dyn_Cloth_volume_l_B03!(ComponentPose=@n68, Alpha=1, GravityOverride:Vector="0,0,500", ExternalForce="0,0,0") -> (Pose) {
    angularDampingOverride=0.99
    angularSpringConstant=800
    bAngularSpring=true
    bOverrideAngularDamping=true
    bUseGravityOverride=true
    gravityScale=0.3
    simulationSpace=BoneRelative
    alphaBoolBlend {blendOption=Linear}
    boundBone {boneName=dyn_Cloth_volume_l_B03}
    componentAppliedLinearAccClamp=<2000 2000 1600>
    componentLinearAccScale=<0.4 0.4 0.4>
    componentLinearVelScale=<0.45 0.45 0.55>
    gravityOverride {z=500}
    physicsBodyDefinitions [{
      collisionType=CustomSphere
      sphereCollisionRadius=2
      boundBone {boneName=dyn_Cloth_volume_l_B03}
      boxExtents=<20 20 20>
      constraintSetup {angularLimitsMax=<10 1 31>, angularLimitsMin=<-10 -1 0>, angularTarget=<0 0 0>}
      localJointOffset=<-5 0.2 -0.1>
     }]
    relativeSpaceBone {boneName=root}
   }
   n70 = Anim Dynamics - Bone: dyn_Cloth_volume_l_C!(ComponentPose=@n69, Alpha=1, GravityOverride:Vector="0,0,500", ExternalForce="0,0,0") -> (Pose) {
    angularDampingOverride=0.99
    angularSpringConstant=800
    bAngularSpring=true
    bOverrideAngularDamping=true
    bUseGravityOverride=true
    gravityScale=0.3
    simulationSpace=BoneRelative
    alphaBoolBlend {blendOption=Linear}
    boundBone {boneName=dyn_Cloth_volume_l_C}
    componentAppliedLinearAccClamp=<2000 2000 1600>
    componentLinearAccScale=<0.5 0.5 0.4>
    componentLinearVelScale=<0.55 0.55 0.55>
    gravityOverride {z=500}
    physicsBodyDefinitions [{
      collisionType=CustomSphere
      sphereCollisionRadius=2
      boundBone {boneName=dyn_Cloth_volume_l_C}
      boxExtents=<60 60 60>
      constraintSetup {angularLimitsMax=<5 5 20>, angularLimitsMin=<-1 -2 0>, angularTarget=<0 0 0>}
      localJointOffset=<-5 0.2 -0.1>
     }]
    relativeSpaceBone {boneName=root}
   }
   n71 = Anim Dynamics - Bone: dyn_Cloth_volume_l_C(ComponentPose=@n70, Alpha=1, GravityOverride="0,0,500", ExternalForce="0,0,0") {
    angularDampingOverride=0.99
    angularSpringConstant=800
    bAngularSpring=true
    bOverrideAngularDamping=true
    bUseGravityOverride=true
    gravityScale=0.3
    simulationSpace=BoneRelative
    alphaBoolBlend {blendOption=Linear}
    boundBone {boneName=dyn_Cloth_volume_l_C}
    componentAppliedLinearAccClamp=<2000 2000 1600>
    componentLinearAccScale=<0.5 0.5 0.4>
    componentLinearVelScale=<0.55 0.55 0.55>
    gravityOverride {z=500}
    physicsBodyDefinitions [{
      collisionType=CustomSphere
      sphereCollisionRadius=2
      boundBone {boneName=dyn_Cloth_volume_l_C}
      boxExtents=<50 50 50>
      constraintSetup {angularLimitsMax=<5 5 20>, angularLimitsMin=<-1 -2 0>, angularTarget=<0 0 0>}
      localJointOffset=<-5 0.2 -0.1>
     }]
    relativeSpaceBone {boneName=root}
   }
   n72 = Anim Dynamics - Bone: dyn_Cloth_volume_l_C02!(ComponentPose=@n71, Alpha=1, GravityOverride:Vector="0,0,500", ExternalForce="0,0,0") -> (Pose) {
    angularDampingOverride=0.99
    angularSpringConstant=800
    bAngularSpring=true
    bOverrideAngularDamping=true
    bUseGravityOverride=true
    gravityScale=0.3
    simulationSpace=BoneRelative
    alphaBoolBlend {blendOption=Linear}
    boundBone {boneName=dyn_Cloth_volume_l_C02}
    componentAppliedLinearAccClamp=<2000 2000 1600>
    componentLinearAccScale=<0.5 0.5 0.4>
    componentLinearVelScale=<0.55 0.55 0.55>
    gravityOverride {z=500}
    physicsBodyDefinitions [{
      collisionType=CustomSphere
      sphereCollisionRadius=2
      boundBone {boneName=dyn_Cloth_volume_l_C02}
      boxExtents=<40 40 40>
      constraintSetup {angularLimitsMax=<5 5 20>, angularLimitsMin=<-1 -2 0>, angularTarget=<0 0 0>}
      localJointOffset=<-5 0.2 -0.1>
     }]
    relativeSpaceBone {boneName=root}
   }
   n73 = Anim Dynamics - Bone: dyn_Cloth_volume_l_C03!(ComponentPose=@n72, Alpha=1, GravityOverride:Vector="0,0,500", ExternalForce="0,0,0") -> (Pose) {
    angularDampingOverride=0.99
    angularSpringConstant=800
    bAngularSpring=true
    bOverrideAngularDamping=true
    bUseGravityOverride=true
    gravityScale=0.3
    simulationSpace=BoneRelative
    alphaBoolBlend {blendOption=Linear}
    boundBone {boneName=dyn_Cloth_volume_l_C03}
    componentAppliedLinearAccClamp=<2000 2000 1600>
    componentLinearAccScale=<0.5 0.5 0.4>
    componentLinearVelScale=<0.55 0.55 0.55>
    gravityOverride {z=500}
    physicsBodyDefinitions [{
      collisionType=CustomSphere
      sphereCollisionRadius=2
      boundBone {boneName=dyn_Cloth_volume_l_C03}
      boxExtents=<30 30 30>
      constraintSetup {angularLimitsMax=<5 5 20>, angularLimitsMin=<-1 -2 0>, angularTarget=<0 0 0>}
      localJointOffset=<-5 0.2 -0.1>
     }]
    relativeSpaceBone {boneName=root}
   }
   n74 = Anim Dynamics - Bone: dyn_Cloth_volume_l_C04!(ComponentPose=@n73, Alpha=1, GravityOverride:Vector="0,0,500", ExternalForce="0,0,0") -> (Pose) {
    angularDampingOverride=0.99
    angularSpringConstant=800
    bAngularSpring=true
    bOverrideAngularDamping=true
    bUseGravityOverride=true
    bUseSphericalLimits=true
    gravityScale=0.3
    simulationSpace=BoneRelative
    alphaBoolBlend {blendOption=Linear}
    boundBone {boneName=dyn_Cloth_volume_l_C04}
    componentAppliedLinearAccClamp=<2000 2000 1600>
    componentLinearAccScale=<0.5 0.5 0.4>
    componentLinearVelScale=<0.55 0.55 0.55>
    gravityOverride {z=500}
    physicsBodyDefinitions [{
      collisionType=CustomSphere
      sphereCollisionRadius=5
      boundBone {boneName=dyn_Cloth_volume_l_C04}
      boxExtents=<20 20 20>
      constraintSetup {angularLimitsMax=<5 5 20>, angularLimitsMin=<-1 -2 0>, angularTarget=<0 0 0>}
      localJointOffset=<-5 0.2 -0.1>
     }]
    relativeSpaceBone {boneName=root}
    sphericalLimits [
     {limitRadius=8, drivingBone {boneName=foot_l}, sphereLocalOffset=<-4 0 0>}
     {limitRadius=8, drivingBone {boneName=foot_r}, sphereLocalOffset=<4 0 0>}
    ]
   }
   n75 = Anim Dynamics - Bone: dyn_Cloth_volume_r_A02!(ComponentPose=@n74, Alpha=1, GravityOverride:Vector="0,0,-980", ExternalForce="0,0,0") -> (Pose) {
    angularDampingOverride=0.99
    angularSpringConstant=800
    bAngularSpring=true
    bOverrideAngularDamping=true
    bUseGravityOverride=true
    gravityScale=0.3
    simulationSpace=BoneRelative
    alphaBoolBlend {blendOption=Linear}
    boundBone {boneName=dyn_Cloth_volume_r_A02}
    componentAppliedLinearAccClamp=<2000 2000 1600>
    componentLinearAccScale=<0.4 0.4 0.4>
    componentLinearVelScale=<0.45 0.45 0.55>
    gravityOverride {z=-980}
    physicsBodyDefinitions [{
      collisionType=CustomSphere
      sphereCollisionRadius=2
      boundBone {boneName=dyn_Cloth_volume_r_A02}
      boxExtents=<33 33 33>
      constraintSetup {twistAxis=AxisY, angularLimitsMax=<0 0 31>, angularLimitsMin=<0 0 0>, angularTarget=<0 0 0>}
      localJointOffset=<-5 0.2 -0.1>
     }]
    relativeSpaceBone {boneName=root}
   }
   n76 = Anim Dynamics - Bone: dyn_Cloth_volume_r_A03!(ComponentPose=@n75, Alpha=1, GravityOverride:Vector="0,0,-980", ExternalForce="0,0,0") -> (Pose) {
    angularDampingOverride=0.99
    angularSpringConstant=800
    bAngularSpring=true
    bOverrideAngularDamping=true
    bUseGravityOverride=true
    gravityScale=0.3
    simulationSpace=BoneRelative
    alphaBoolBlend {blendOption=Linear}
    boundBone {boneName=dyn_Cloth_volume_r_A03}
    componentAppliedLinearAccClamp=<2000 2000 1600>
    componentLinearAccScale=<0.4 0.4 0.4>
    componentLinearVelScale=<0.45 0.45 0.55>
    gravityOverride {z=-980}
    physicsBodyDefinitions [{
      collisionType=CustomSphere
      sphereCollisionRadius=2
      boundBone {boneName=dyn_Cloth_volume_r_A03}
      boxExtents=<27 27 27>
      constraintSetup {twistAxis=AxisY, angularLimitsMax=<0 0 31>, angularLimitsMin=<0 0 0>, angularTarget=<0 0 0>}
      localJointOffset=<-5 0.2 -0.1>
     }]
    relativeSpaceBone {boneName=root}
   }
   n77 = Anim Dynamics - Bone: dyn_Cloth_volume_r_A04!(ComponentPose=@n76, Alpha=1, GravityOverride:Vector="0,0,-980", ExternalForce="0,0,0") -> (Pose) {
    angularDampingOverride=0.99
    angularSpringConstant=800
    bAngularSpring=true
    bOverrideAngularDamping=true
    bUseGravityOverride=true
    gravityScale=0.3
    simulationSpace=BoneRelative
    alphaBoolBlend {blendOption=Linear}
    boundBone {boneName=dyn_Cloth_volume_r_A04}
    componentAppliedLinearAccClamp=<2000 2000 1600>
    componentLinearAccScale=<0.4 0.4 0.4>
    componentLinearVelScale=<0.45 0.45 0.55>
    gravityOverride {z=-980}
    physicsBodyDefinitions [{
      collisionType=CustomSphere
      sphereCollisionRadius=2
      boundBone {boneName=dyn_Cloth_volume_r_A04}
      boxExtents=<27 27 27>
      constraintSetup {twistAxis=AxisY, angularLimitsMax=<0 0 31>, angularLimitsMin=<0 0 0>, angularTarget=<0 0 0>}
      localJointOffset=<-5 0.2 -0.1>
     }]
    relativeSpaceBone {boneName=root}
   }
   n78 = Anim Dynamics - Bone: dyn_Cloth_volume_r_B!(ComponentPose=@n77, Alpha=1, GravityOverride:Vector="0,0,-980", ExternalForce="0,0,0") -> (Pose) {
    angularDampingOverride=0.99
    angularSpringConstant=800
    bAngularSpring=true
    bOverrideAngularDamping=true
    bUseGravityOverride=true
    gravityScale=0.3
    simulationSpace=BoneRelative
    alphaBoolBlend {blendOption=Linear}
    boundBone {boneName=dyn_Cloth_volume_r_B}
    componentAppliedLinearAccClamp=<2000 2000 1600>
    componentLinearAccScale=<0.4 0.4 0.4>
    componentLinearVelScale=<0.45 0.45 0.55>
    gravityOverride {z=-980}
    physicsBodyDefinitions [{
      collisionType=CustomSphere
      sphereCollisionRadius=2
      boundBone {boneName=dyn_Cloth_volume_r_B}
      boxExtents=<50 50 50>
      constraintSetup {angularLimitsMax=<5 5 31>, angularLimitsMin=<0 -5 0>, angularTarget=<0 0 0>}
      localJointOffset=<-5 0.2 -0.1>
     }]
    relativeSpaceBone {boneName=root}
   }
   n79 = Anim Dynamics - Bone: dyn_Cloth_volume_r_B01!(ComponentPose=@n78, Alpha=1, GravityOverride:Vector="0,0,-980", ExternalForce="0,0,0") -> (Pose) {
    angularDampingOverride=0.99
    angularSpringConstant=800
    bAngularSpring=true
    bOverrideAngularDamping=true
    bUseGravityOverride=true
    gravityScale=0.3
    simulationSpace=BoneRelative
    alphaBoolBlend {blendOption=Linear}
    boundBone {boneName=dyn_Cloth_volume_r_B01}
    componentAppliedLinearAccClamp=<2000 2000 1600>
    componentLinearAccScale=<0.4 0.4 0.4>
    componentLinearVelScale=<0.45 0.45 0.55>
    gravityOverride {z=-980}
    physicsBodyDefinitions [{
      collisionType=CustomSphere
      sphereCollisionRadius=2
      boundBone {boneName=dyn_Cloth_volume_r_B01}
      boxExtents=<33 33 33>
      constraintSetup {angularLimitsMax=<5 5 31>, angularLimitsMin=<0 -5 0>, angularTarget=<0 0 0>}
      localJointOffset=<-5 0.2 -0.1>
     }]
    relativeSpaceBone {boneName=root}
   }
   n80 = Anim Dynamics - Bone: dyn_Cloth_volume_r_B02!(ComponentPose=@n79, Alpha=1, GravityOverride:Vector="0,0,-980", ExternalForce="0,0,0") -> (Pose) {
    angularDampingOverride=0.99
    angularSpringConstant=800
    bAngularSpring=true
    bOverrideAngularDamping=true
    bUseGravityOverride=true
    gravityScale=0.3
    simulationSpace=BoneRelative
    alphaBoolBlend {blendOption=Linear}
    boundBone {boneName=dyn_Cloth_volume_r_B02}
    componentAppliedLinearAccClamp=<2000 2000 1600>
    componentLinearAccScale=<0.4 0.4 0.4>
    componentLinearVelScale=<0.45 0.45 0.55>
    gravityOverride {z=-980}
    physicsBodyDefinitions [{
      collisionType=CustomSphere
      sphereCollisionRadius=2
      boundBone {boneName=dyn_Cloth_volume_r_B02}
      boxExtents=<27 27 27>
      constraintSetup {angularLimitsMax=<5 5 31>, angularLimitsMin=<0 -5 0>, angularTarget=<0 0 0>}
      localJointOffset=<-5 0.2 -0.1>
     }]
    relativeSpaceBone {boneName=root}
   }
   n81 = Anim Dynamics - Bone: dyn_Cloth_volume_r_B03!(ComponentPose=@n80, Alpha=1, GravityOverride:Vector="0,0,-980", ExternalForce="0,0,0") -> (Pose) {
    angularDampingOverride=0.99
    angularSpringConstant=800
    bAngularSpring=true
    bOverrideAngularDamping=true
    bUseGravityOverride=true
    gravityScale=0.3
    simulationSpace=BoneRelative
    alphaBoolBlend {blendOption=Linear}
    boundBone {boneName=dyn_Cloth_volume_r_B03}
    componentAppliedLinearAccClamp=<2000 2000 1600>
    componentLinearAccScale=<0.4 0.4 0.4>
    componentLinearVelScale=<0.45 0.45 0.55>
    gravityOverride {z=-980}
    physicsBodyDefinitions [{
      collisionType=CustomSphere
      sphereCollisionRadius=2
      boundBone {boneName=dyn_Cloth_volume_r_B03}
      boxExtents=<20 20 20>
      constraintSetup {angularLimitsMax=<5 5 31>, angularLimitsMin=<0 -5 0>, angularTarget=<0 0 0>}
      localJointOffset=<-5 0.2 -0.1>
     }]
    relativeSpaceBone {boneName=root}
   }
   n82 = Anim Dynamics - Bone: dyn_Cloth_volume_r_C!(ComponentPose=@n81, Alpha=1, GravityOverride:Vector="0,0,-980", ExternalForce="0,0,0") -> (Pose) {
    angularDampingOverride=0.99
    angularSpringConstant=800
    bAngularSpring=true
    bOverrideAngularDamping=true
    bUseGravityOverride=true
    gravityScale=0.3
    simulationSpace=BoneRelative
    alphaBoolBlend {blendOption=Linear}
    boundBone {boneName=dyn_Cloth_volume_r_C}
    componentAppliedLinearAccClamp=<2000 2000 1600>
    componentLinearAccScale=<0.5 0.5 0.4>
    componentLinearVelScale=<0.55 0.55 0.55>
    gravityOverride {z=-980}
    physicsBodyDefinitions [{
      collisionType=CustomSphere
      sphereCollisionRadius=2
      boundBone {boneName=dyn_Cloth_volume_r_C}
      boxExtents=<45 45 45>
      constraintSetup {angularLimitsMax=<1 2 20>, angularLimitsMin=<-5 -5 0>, angularTarget=<0 0 0>}
      localJointOffset=<-5 0.2 -0.1>
     }]
    relativeSpaceBone {boneName=root}
   }
   n83 = Anim Dynamics - Bone: dyn_Cloth_volume_r_C01!(ComponentPose=@n82, Alpha=1, GravityOverride:Vector="0,0,-980", ExternalForce="0,0,0") -> (Pose) {
    angularDampingOverride=0.99
    angularSpringConstant=800
    bAngularSpring=true
    bOverrideAngularDamping=true
    bUseGravityOverride=true
    gravityScale=0.3
    simulationSpace=BoneRelative
    alphaBoolBlend {blendOption=Linear}
    boundBone {boneName=dyn_Cloth_volume_r_C01}
    componentAppliedLinearAccClamp=<2000 2000 1600>
    componentLinearAccScale=<0.5 0.5 0.4>
    componentLinearVelScale=<0.55 0.55 0.55>
    gravityOverride {z=-980}
    physicsBodyDefinitions [{
      collisionType=CustomSphere
      sphereCollisionRadius=2
      boundBone {boneName=dyn_Cloth_volume_r_C01}
      boxExtents=<40 40 40>
      constraintSetup {angularLimitsMax=<1 2 20>, angularLimitsMin=<-5 -5 0>, angularTarget=<0 0 0>}
      localJointOffset=<-5 0.2 -0.1>
     }]
    relativeSpaceBone {boneName=root}
   }
   n84 = Anim Dynamics - Bone: dyn_Cloth_volume_r_C02!(ComponentPose=@n83, Alpha=1, GravityOverride:Vector="0,0,-980", ExternalForce="0,0,0") -> (Pose) {
    angularDampingOverride=0.99
    angularSpringConstant=800
    bAngularSpring=true
    bOverrideAngularDamping=true
    bUseGravityOverride=true
    gravityScale=0.3
    simulationSpace=BoneRelative
    alphaBoolBlend {blendOption=Linear}
    boundBone {boneName=dyn_Cloth_volume_r_C02}
    componentAppliedLinearAccClamp=<2000 2000 1600>
    componentLinearAccScale=<0.5 0.5 0.4>
    componentLinearVelScale=<0.55 0.55 0.55>
    gravityOverride {z=-980}
    physicsBodyDefinitions [{
      collisionType=CustomSphere
      sphereCollisionRadius=2
      boundBone {boneName=dyn_Cloth_volume_r_C02}
      boxExtents=<33 33 33>
      constraintSetup {angularLimitsMax=<1 2 20>, angularLimitsMin=<-5 -5 0>, angularTarget=<0 0 0>}
      localJointOffset=<-5 0.2 -0.1>
     }]
    relativeSpaceBone {boneName=root}
   }
   n85 = Anim Dynamics - Bone: dyn_Cloth_volume_r_C03!(ComponentPose=@n84, Alpha=1, GravityOverride:Vector="0,0,-980", ExternalForce="0,0,0") -> (Pose) {
    angularDampingOverride=0.99
    angularSpringConstant=800
    bAngularSpring=true
    bOverrideAngularDamping=true
    bUseGravityOverride=true
    gravityScale=0.3
    simulationSpace=BoneRelative
    alphaBoolBlend {blendOption=Linear}
    boundBone {boneName=dyn_Cloth_volume_r_C03}
    componentAppliedLinearAccClamp=<2000 2000 1600>
    componentLinearAccScale=<0.5 0.5 0.4>
    componentLinearVelScale=<0.55 0.55 0.55>
    gravityOverride {z=-980}
    physicsBodyDefinitions [{
      collisionType=CustomSphere
      sphereCollisionRadius=2
      boundBone {boneName=dyn_Cloth_volume_r_C03}
      boxExtents=<27 27 27>
      constraintSetup {angularLimitsMax=<1 2 20>, angularLimitsMin=<-5 -5 0>, angularTarget=<0 0 0>}
      localJointOffset=<-5 0.2 -0.1>
     }]
    relativeSpaceBone {boneName=root}
   }
   n86 = Anim Dynamics - Bone: dyn_Cloth_volume_r_C04!(ComponentPose=@n85, Alpha=1, GravityOverride:Vector="0,0,-980", ExternalForce="0,0,0") -> (Pose) {
    angularDampingOverride=0.99
    angularSpringConstant=800
    bAngularSpring=true
    bOverrideAngularDamping=true
    bUseGravityOverride=true
    bUseSphericalLimits=true
    gravityScale=0.3
    simulationSpace=BoneRelative
    alphaBoolBlend {blendOption=Linear}
    boundBone {boneName=dyn_Cloth_volume_r_C04}
    componentAppliedLinearAccClamp=<2000 2000 1600>
    componentLinearAccScale=<0.5 0.5 0.4>
    componentLinearVelScale=<0.55 0.55 0.55>
    gravityOverride {z=-980}
    physicsBodyDefinitions [{
      collisionType=CustomSphere
      sphereCollisionRadius=5
      boundBone {boneName=dyn_Cloth_volume_r_C04}
      boxExtents=<20 20 20>
      constraintSetup {angularLimitsMax=<1 2 20>, angularLimitsMin=<-5 -5 0>, angularTarget=<0 0 0>}
      localJointOffset=<-5 0.2 -0.1>
     }]
    relativeSpaceBone {boneName=root}
    sphericalLimits [
     {limitRadius=8, drivingBone {boneName=foot_l}, sphereLocalOffset=<-4 0 0>}
     {limitRadius=8, drivingBone {boneName=foot_r}, sphereLocalOffset=<4 0 0>}
    ]
   }
   n87 = "Constraint - dyn_Cloth_volume_bk (PP)"(ComponentPose=@n86, Alpha=1, ConstraintWeights_0=0.5, ConstraintWeights_1=0.5) {
    alphaBoolBlend {blendOption=Linear}
    boneToModify {boneName=dyn_Cloth_volume_bk}
    constraintSetup [{targetBone {boneName=dyn_Cloth_volume_l_C}}, {targetBone {boneName=dyn_Cloth_volume_r_C}}]
    constraintWeights [1,1]
   }
   n88 = "Constraint - dyn_Cloth_volume_bk01 (PP)"(ComponentPose=@n87, Alpha=1, ConstraintWeights_0=0.5, ConstraintWeights_1=0.5) {
    alphaBoolBlend {blendOption=Linear}
    boneToModify {boneName=dyn_Cloth_volume_bk01}
    constraintSetup [{targetBone {boneName=dyn_Cloth_volume_l_C01}}, {targetBone {boneName=dyn_Cloth_volume_r_C01}}]
    constraintWeights [1,1]
   }
   n89 = "Constraint - dyn_Cloth_volume_bk02 (PP)"(ComponentPose=@n88, Alpha=1, ConstraintWeights_0=0.5, ConstraintWeights_1=0.5) {
    alphaBoolBlend {blendOption=Linear}
    boneToModify {boneName=dyn_Cloth_volume_bk02}
    constraintSetup [{targetBone {boneName=dyn_Cloth_volume_l_C02}}, {targetBone {boneName=dyn_Cloth_volume_r_C02}}]
    constraintWeights [1,1]
   }
   Save cached pose 'MovingPose'(Pose=Component To Local(ComponentPose=@n89)) {cacheName=MovingPose}
   n90 = "Constraint - dyn_Cloth_volume_r_C02 (RR)"(ComponentPose=Local To Component(LocalPose=Use cached pose 'SkirtCachePose'()), Alpha=1, ConstraintWeights_0=0.4, ConstraintWeights_1=0.6) {
    alphaBoolBlend {blendOption=Linear}
    boneToModify {boneName=dyn_Cloth_volume_r_C02}
    constraintSetup [{transformType=Rotation, targetBone {boneName=calf_r}}, {transformType=Rotation, targetBone {boneName=dyn_Cloth_volume_r_C01}}]
    constraintWeights [1,1]
   }
   n91 = "Constraint - dyn_Cloth_volume_l_C02 (RR)"(ComponentPose=@n90, Alpha=1, ConstraintWeights_0=0.4, ConstraintWeights_1=0.6) {
    alphaBoolBlend {blendOption=Linear}
    boneToModify {boneName=dyn_Cloth_volume_l_C02}
    constraintSetup [{transformType=Rotation, targetBone {boneName=calf_l}}, {transformType=Rotation, targetBone {boneName=dyn_Cloth_volume_l_C01}}]
    constraintWeights [1,1]
   }
   n92 = "Constraint - dyn_Cloth_volume_r_C03 (RR)"!(ComponentPose=@n91, Alpha=1, ConstraintWeights_0=0.4, ConstraintWeights_1:float=0.6) -> (Pose) {
    alphaBoolBlend {blendOption=Linear}
    boneToModify {boneName=dyn_Cloth_volume_r_C03}
    constraintSetup [{transformType=Rotation, targetBone {boneName=calf_r}}, {transformType=Rotation, targetBone {boneName=dyn_Cloth_volume_r_C02}}]
    constraintWeights [1,1]
   }
   n93 = "Constraint - dyn_Cloth_volume_l_C03 (RR)"!(ComponentPose=@n92, Alpha=1, ConstraintWeights_0=0.4, ConstraintWeights_1:float=0.6) -> (Pose) {
    alphaBoolBlend {blendOption=Linear}
    boneToModify {boneName=dyn_Cloth_volume_l_C03}
    constraintSetup [{transformType=Rotation, targetBone {boneName=calf_l}}, {transformType=Rotation, targetBone {boneName=dyn_Cloth_volume_l_C02}}]
    constraintWeights [1,1]
   }
   n94 = Component To Local(ComponentPose=@n93)
   n95 = "Constraint - dyn_Cloth_volume_r_C03 (RR)"(ComponentPose=Local To Component(LocalPose=Use cached pose 'SkirtCachePose'()), Alpha=1, ConstraintWeights_0=0.6, ConstraintWeights_1=0.4) {
    alphaBoolBlend {blendOption=Linear}
    boneToModify {boneName=dyn_Cloth_volume_r_C03}
    constraintSetup [{transformType=Rotation, targetBone {boneName=calf_r}}, {transformType=Rotation, targetBone {boneName=dyn_Cloth_volume_r_C02}}]
    constraintWeights [1,1]
   }
   n96 = "Constraint - dyn_Cloth_volume_l_C03 (RR)"(ComponentPose=@n95, Alpha=1, ConstraintWeights_0=0.6, ConstraintWeights_1=0.4) {
    alphaBoolBlend {blendOption=Linear}
    boneToModify {boneName=dyn_Cloth_volume_l_C03}
    constraintSetup [{transformType=Rotation, targetBone {boneName=calf_l}}, {transformType=Rotation, targetBone {boneName=dyn_Cloth_volume_l_C02}}]
    constraintWeights [1,1]
   }
   n97 = Component To Local(ComponentPose=@n96)
   n98 = Get Not DiveFloatingAndDiveMode() -> (Not DiveFloatingAndDiveMode:bool)
   n99 = "Constraint - dyn_Cloth_volume_bk (PP)"(
    ComponentPose=Local To Component(LocalPose=Blend Poses by bool(BlendPose_0=@n97, BlendPose_1=@n94, BlendTime_0=0.5, bActiveValue=@n98) {blendType=Linear})
    Alpha=1
    ConstraintWeights_0=0.5
    ConstraintWeights_1=0.5
   ) {
    alphaBoolBlend {blendOption=Linear}
    boneToModify {boneName=dyn_Cloth_volume_bk}
    constraintSetup [{targetBone {boneName=dyn_Cloth_volume_l_C}}, {targetBone {boneName=dyn_Cloth_volume_r_C}}]
    constraintWeights [1,1]
   }
   n100 = "Constraint - dyn_Cloth_volume_bk01 (PP)"(ComponentPose=@n99, Alpha=1, ConstraintWeights_0=0.5, ConstraintWeights_1=0.5) {
    alphaBoolBlend {blendOption=Linear}
    boneToModify {boneName=dyn_Cloth_volume_bk01}
    constraintSetup [{targetBone {boneName=dyn_Cloth_volume_l_C01}}, {targetBone {boneName=dyn_Cloth_volume_r_C01}}]
    constraintWeights [1,1]
   }
   n101 = "Constraint - dyn_Cloth_volume_bk02 (PP)"(ComponentPose=@n100, Alpha=1, ConstraintWeights_0=0.5, ConstraintWeights_1=0.5) {
    alphaBoolBlend {blendOption=Linear}
    boneToModify {boneName=dyn_Cloth_volume_bk02}
    constraintSetup [{targetBone {boneName=dyn_Cloth_volume_l_C02}}, {targetBone {boneName=dyn_Cloth_volume_r_C02}}]
    constraintWeights [1,1]
   }
   n102 = RigidBody(ComponentPose=@n101, bAlphaBoolEnabled=Get DivingAndCape() -> (DivingAndCape:bool), OverrideWorldGravity="0,0,-980") {
    alphaInputType=Bool
    bClampLinearTranslationLimitToRefPose=true
    bOverrideWorldGravity=true
    overridePhysicsAsset=&/M_MED_PageTruffle_Dyning_Physics
    simulationSpace=BaseBoneSpace
    alphaBoolBlend {blendOption=Linear}
    baseBoneRef {boneName=gravity}
    componentAppliedLinearAccClamp=<3000 3000 2700>
    componentLinearAccScale=<0.01 0.6 0.6>
    componentLinearVelScale=<0.08 0.4 0.4>
    overrideWorldGravity {z=-980}
   }
   Save cached pose 'DBNOPose'(Pose=Component To Local(ComponentPose=@n102)) {cacheName=DBNOPose}
   Use cached pose 'SkirtCachePose'()
  ]}
 EventGraph {nodes [
   n1 = Event BlueprintUpdateAnimation()
   > n2 = Sequence!(execute) -> (then_0, then_1, then_2:Exec, then_3:Exec, then_4:Exec, then_5:Exec, then_6:Exec, then_7:Exec)
   n3 = Cloth Controller Function DEPRECATED!(
    execute=@n2.then_1
    CharacterPartAnimInstance:Custom Character Part Anim Instance Object Reference=Self-Reference()
    ClothingAssetName:String=M_MED_PageTruffle_Clothing_0
    Cloth Paramaters:Cloth Params Struct Structure=Get Cloth Paramaters_0() -> (Cloth Paramaters_0:Cloth Params Struct Structure)
    Cloth Core Settings:Cloth Core Setting Struct Structure=Get Cloth Core Settings_0() -> (Cloth Core Settings_0:Cloth Core Setting Struct Structure)
    ClothUpAxis:Vector="1,0,0"
    OnGround_Moving_SpeedLimit:double=2
    OnGround_Moving_AngularVel_Max:double=0.05
    OnGround_Moving_FictitiousVel_Max:double=0.6
    LinVelScale_FwdBwd_Clamp:double=0
    LinVelScale_UpDwn_Clamp:double=0.7
    DeltaTime:double=Get DeltaTimeLocal() -> (DeltaTimeLocal:double)
    GameTimeInSeconds:double=GetGameTimeInSeconds()
    WindGustsParams:Wind Gust Params Struct Structure=Get Wind Gust Params() -> (Wind Gust Params:Wind Gust Params Struct Structure)
    WindGustLocalSettings:Wind Gust Local Setting Struct Structure=Get Gust Local Settings() -> (Gust Local Settings:Wind Gust Local Setting Struct Structure)
   ) -> (
    then
    CurrentStateOut:String
    Cloth Settings Out:Cloth Setting Struct Structure
    Cloth Core Settings Out:Cloth Core Setting Struct Structure
    WindGustSettings Out:Wind Gust Setting Struct Structure
    WindGustLocalSettings Out:Wind Gust Local Setting Struct Structure
    GustStrength - WindVelocityOut:Vector
    LinearVelVector-DELETE LATER:Vector
   )
   > Set Gust Local Settings(Gust Local Settings:Wind Gust Local Setting Struct Structure=@n3.WindGustLocalSettings Out) -> (Output_Get:Wind Gust Local Setting Struct Structure)
   Set is sliding(execute=@n2.then_2, is sliding:double="To Float (Boolean)"(InBool=Get bIsSliding())) -> (Output_Get:double)
   n4 = Get bIsSkydiveFloating()
   > Set Not DiveFloatingAndDiveMode(Not DiveFloatingAndDiveMode:bool=AND Boolean(A=NOT Boolean(A=@n4), B=Get bIsSkydiveDiveMode())) -> (Output_Get:bool)
   n5 = Get thigh_shape() -> (thigh_shape:bool)
   n6 = Get bIsMoving2D()
   > Set CrouchingAndNotMovingAndNotShape(CrouchingAndNotMovingAndNotShape:bool=AND Boolean!(A=Get bIsCrouching(), B=NOT Boolean(A=@n6), C:bool=NOT Boolean(A=@n5)) -> (ReturnValue)) -> (Output_Get:bool)
   n7 = Get bIsSkyDiving()
   n8 = Get bIsCrouching()
   n9 = Get bIsPlayingEmote()
   > Set MovingAnd NotCrouchNotDivingNotEmote(MovingAnd NotCrouchNotDivingNotEmote:bool=AND Boolean!(A=Get bIsMoving2D(), B=NOT Boolean(A=@n8), C=NOT Boolean(A=@n7), D:bool=NOT Boolean(A=@n9)) -> (ReturnValue)) -> (Output_Get:bool)
   > Set DivingAndCape(DivingAndCape:bool=AND Boolean!(A=Get bIsSkyDiving(), B=Get is_WearingCape()) -> (ReturnValue)) -> (Output_Get:bool)
   > Set DivingOrDBNO(DivingOrDBNO:bool=OR Boolean(A=Get bIsSkyDiving(), B=Get bIsDBNO())) -> (Output_Get:bool)
   > Set DBNO_Alpha(DBNO_Alpha:double=SelectFloat(A=0.95, B=0, bPickA=Get bIsDBNO())) -> (Output_Get:double)
   n10 = Get Body Mesh() -> (Body Mesh:Skeletal Mesh Component Object Reference)
   Set thigh_shape_fn_r(
    execute=@n2.then_3
    thigh_shape_fn_r:double=GetAlphaFromReferenceTransformDelta(SkeletalMeshComponent=@n10, BoneName=thigh_r, BaseName=pelvis, InputRange=(X=0,Y=-110), TwistAxis="0,0,1")
   ) -> (Output_Get:double)
   > Set thigh_shape_fn_r_1(thigh_shape_fn_r_1:double=GetAlphaFromReferenceTransformDelta(SkeletalMeshComponent=@n10, BoneName=calf_r, BaseName=thigh_r, InputRange=(X=0,Y=150), TwistAxis="0,0,1")) -> (Output_Get:double)
   > Set thigh_shape(thigh_shape:bool="float < float"(A=Get thigh_shape_fn_r_1() -> (thigh_shape_fn_r_1:double), B=0.8)) -> (Output_Get:bool)
   > Set thigh_shape_fn_r_0(thigh_shape_fn_r_0:double=float - float(A=1.5, B=Get thigh_shape_fn_r() -> (thigh_shape_fn_r:double))) -> (Output_Get:double)
   Set DeltaTimeLocal(execute=@n2.then_4, DeltaTimeLocal:double=@n1.DeltaTimeX) -> (Output_Get:double)
   n11 = Get bIsSkyDiving()
   n12 = Get bIsPlayingEmote()
   Set isFrontEndAndNotEmote(
    execute=@n2.then_5
    isFrontEndAndNotEmote:bool=AND Boolean!(A=Get bIsInFrontEnd(), B=NOT Boolean(A=@n12), C:bool=NOT Boolean(A=@n11)) -> (ReturnValue)
   ) -> (Output_Get:bool)
   n13 = Gravity Override State Function DEPRECATED(
    execute=@n2.then_0
    CharacterPartAnimInstance=Self-Reference()
    GameTimeInSeconds=GetGameTimeInSeconds()
    GravityOverrideParams=Get Gravity Override Params()
   )
   > Set Gravity Override Cape(Gravity Override Cape:Vector=@n13) -> (Output_Get:Vector)
   n14 = Event OnCharacterCustomizationLoaded()
   > n15 = Sequence!(execute) -> (then_0, then_1, then_2, then_3, then_4)
   > Set Character Mesh(Character Mesh:Skeletal Mesh Component Object Reference) -> (Output_Get:Skeletal Mesh Component Object Reference)
   Set is_WearingCape(execute=@n15.then_1, is_WearingCape:bool=true) -> (Output_Get:bool)
   n16 = CheckForBackbling()
   > Set isWearingPackBool(isWearingPackBool:bool=OR Boolean!(A=@n16, B=@n16.bIsBackblingACape, C:bool=@n16.bIsBackblingLong) -> (ReturnValue)) -> (Output_Get:bool)
   > Branch(Condition=OR Boolean(A=@n16, B=@n16.bIsBackblingACape, C=@n16.bIsBackblingLong))
   > Set is_WearingCape(is_WearingCape=false)
   > n17 = Branch(Condition=NOT Boolean(A=Get bIsSkyDiving()))
   > Set is_WearingCape(is_WearingCape=true)
   Set is_WearingCape(execute=@n17.else, is_WearingCape=NOT Boolean(A=@n16.bIsBackblingLong))
   Set Body Mesh(
    execute=@n15.then_2
    Body Mesh:Skeletal Mesh Component Object Reference=Get Array Item with EFort Custom Part Type(Array=@n14.CustomMeshParts)
   ) -> (Output_Get:Skeletal Mesh Component Object Reference)
   GetSkeletalMesh!(self:Custom Character Part Object Reference=Get Array Item with EFort Custom Part Type(Array:Array of Custom Character Part Object References=@n14.CustomParts) -> (Output:Custom Character Part Object Reference)) -> (ReturnValue:Skeletal Mesh Object Reference)
   DrawVectorFromJointFunction!(
    execute
    SkelMeshComponent:Scene Component Object Reference
    JointName:Name=pelvis
    LineEnd:Vector=Break Cloth Setting Struct!(ClothSettingStruct:"Cloth Setting Struct Structure (by ref)"=@n3.Cloth Settings Out) -> (
     LinearVelocityScale_7_#1:Vector=~
     AngularVelocityScale_6_#2:float=~
     FictitiousAngularScale_10_#3:float=~
     GravityOverride_15_#4:Vector=~
     DragCoefficient_23_#5:Vector 2D Structure
     LiftCoefficient_24_#6:Vector 2D Structure
     JointName_27_#7:Name
    ) {is=title K2Node_BreakStruct}.GravityOverride_15_#4
    JointOffset:Vector="0, 0, 0"
    LineColor:Linear Color Structure=(R=1,G=0,B=0,A=1)
    Thickness:double=2
    Foreground:EDrawDebugSceneDepthPriorityGroup Enum=Foreground
   ) -> (then)
   NOT Boolean(A=Get bClothEnabled())
   Set is_WearingCape(is_WearingCape=false)
  ]}
]
variables [
 Body Mesh {type=SkeletalMeshComponent}
 calf_shape_back_l {type=number}
 calf_shape_back_r {type=number}
 Character Mesh {type=SkeletalMeshComponent}
 Cloth Core Settings_0 {
  type=ClothCoreSettingStruct
  default {
   collisionThickness=1
   dampingCoefficient=0.1
   frictionCoefficient=0.3
   localDampingCoefficient=0.1
   numberIterations=2
   numberSubsteps=1
   selfCollisionThickness=0.65
   animDriveDamping=<0 1>
   animDriveStiffness=<0 1>
   areaStiffness=<1 1>
   bendingStiffness=<1 1>
   edgeStiffness=<1 1>
   tetherStiffness=<1 1>
  }
 }
 Cloth Paramaters_0 {
  type=ClothParamsStruct
  default {
   dBNO {
    angularVelocityScale=0.6
    fictitiousAngularScale=0.1
    jointName=pelvis
    dragCoefficient=<0.2 1>
    gravityOverride=<-980 400 0>
    liftCoefficient=<0 0.5>
    linearVelocityScale=<0.3 0.3 0.1>
   }
   dBNO_Carried {
    angularVelocityScale=0.6
    fictitiousAngularScale=0.1
    jointName=pelvis
    dragCoefficient=<0.2 1>
    gravityOverride=<-980 600 0>
    liftCoefficient=<0 0.5>
    linearVelocityScale=<0.3 0.3 0.1>
   }
   dBNO_Impulse {
    angularVelocityScale=0
    fictitiousAngularScale=0
    jointName=pelvis
    dragCoefficient=<0 0.2>
    gravityOverride=<-3000 -2000 0>
    liftCoefficient=<0 0.1>
    linearVelocityScale=<0.3 0.1 0.3>
   }
   dBNO_RollBwd {
    angularVelocityScale=0
    fictitiousAngularScale=0
    jointName=pelvis
    dragCoefficient=<0 0.2>
    gravityOverride=<-2000 0 0>
    liftCoefficient=<0 0.1>
    linearVelocityScale=<0.3 0.1 0.3>
   }
   dBNO_RollFwd {
    angularVelocityScale=0
    fictitiousAngularScale=0
    jointName=pelvis
    dragCoefficient=<0 0.2>
    gravityOverride=<-980 0 0>
    liftCoefficient=<0 0.1>
    linearVelocityScale=<0.3 0.1 0.3>
   }
   dBNO_Tumbling {
    angularVelocityScale=0
    fictitiousAngularScale=0
    jointName=pelvis
    dragCoefficient=<0 0.2>
    gravityOverride=<-2500 -300 0>
    liftCoefficient=<0 0.1>
    linearVelocityScale=<0.3 0.1 0.3>
   }
   emoteOrMelee {
    angularVelocityScale=0.3
    fictitiousAngularScale=1
    jointName=root
    dragCoefficient=<1 1>
    gravityOverride=<0 -400 -980>
    liftCoefficient=<0.1 0.15>
    linearVelocityScale=<0.3 0.3 0.3>
   }
   front_End {
    angularVelocityScale=0.1
    fictitiousAngularScale=0.1
    jointName=pelvis
    dragCoefficient=<0.1 0.5>
    gravityOverride=<-980 -50 50>
    liftCoefficient=<0.1 0.1>
    linearVelocityScale=<0.2 0.2 0.3>
   }
   onGround_Crouch_Moving {
    angularVelocityScale=0.1
    fictitiousAngularScale=0.5
    jointName=root
    dragCoefficient=<0.5 0.5>
    gravityOverride=<0 -500 -700>
    liftCoefficient=<0.1 0.2>
    linearVelocityScale=<0.4 0.4 0.3>
   }
   onGround_Falling {
    angularVelocityScale=0.6
    fictitiousAngularScale=0
    jointName=root
    dragCoefficient=<0 0.2>
    gravityOverride=<0 -500 -980>
    liftCoefficient=<0 0.1>
    linearVelocityScale=<2 1 0.3>
   }
   onGround_Moving {
    angularVelocityScale=0.3
    fictitiousAngularScale=0.1
    jointName=root
    dragCoefficient=<0 0.6>
    gravityOverride=<0 -150 -980>
    liftCoefficient=<0.1 0.3>
    linearVelocityScale=<0.2 0.2 0.1>
   }
   onGround_Standing {
    angularVelocityScale=0.05
    fictitiousAngularScale=0.1
    jointName=root
    dragCoefficient=<0 0.6>
    gravityOverride=<0 0 -980>
    liftCoefficient=<0 0.3>
    linearVelocityScale=<0.15 0.15 0.1>
   }
   rail_Grinding {
    angularVelocityScale=0.2
    fictitiousAngularScale=0
    jointName=root
    dragCoefficient=<0.1 0.2>
    gravityOverride=<50 -500 -400>
    liftCoefficient=<0 0.5>
    linearVelocityScale=<0.1 0.1 0.1>
   }
   riding_Moving {
    angularVelocityScale=0.05
    fictitiousAngularScale=0.1
    jointName=pelvis
    dragCoefficient=<0.1 0.1>
    gravityOverride=<-980 0 0>
    liftCoefficient=<0.1 0.1>
    linearVelocityScale=<0.4 0.6 0.3>
   }
   riding_Moving_Targeting {
    angularVelocityScale=0.05
    fictitiousAngularScale=0.7
    jointName=root
    dragCoefficient=<0.5 0.5>
    gravityOverride=<-980 0 0>
    liftCoefficient=<0.5 0.5>
    linearVelocityScale=<0.2 0.6 0.2>
   }
   riding_Standing {
    angularVelocityScale=0.05
    fictitiousAngularScale=1
    jointName=pelvis
    dragCoefficient=<0.5 0.5>
    gravityOverride=<-980 -450 0>
    liftCoefficient=<0.5 0.5>
    linearVelocityScale=<0.6 0.6 0.3>
   }
   skydive_DivingDown {
    angularVelocityScale=0.1
    fictitiousAngularScale=0
    jointName=pelvis
    dragCoefficient=<0 1>
    gravityOverride=<0 -200 -500>
    liftCoefficient=<0.1 0.5>
    linearVelocityScale=<0.1 0.1 0.1>
   }
   skydive_FloatingDown {
    angularVelocityScale=0.1
    fictitiousAngularScale=0
    jointName=root
    dragCoefficient=<0 1>
    gravityOverride=<0 -600 -150>
    liftCoefficient=<0.1 0.2>
    linearVelocityScale=<0.2 0.2 0.1>
   }
   skydive_GoingUp {
    angularVelocityScale=0.1
    fictitiousAngularScale=0
    jointName=root
    dragCoefficient=<0 1>
    gravityOverride=<0 -800 150>
    liftCoefficient=<0 0.1>
    linearVelocityScale=<0.01 0.01 0.01>
   }
   skydive_Parachuting {
    angularVelocityScale=0.1
    fictitiousAngularScale=0
    jointName=root
    dragCoefficient=<0.1 1>
    gravityOverride=<0 -200 -980>
    liftCoefficient=<0.2 1>
    linearVelocityScale=<0.1 0.1 0.1>
   }
   skydive_Surfing {
    angularVelocityScale=0.1
    fictitiousAngularScale=0
    jointName=root
    dragCoefficient=<0.1 1>
    gravityOverride=<200 -50 -800>
    liftCoefficient=<0.2 1>
    linearVelocityScale=<0.1 0.1 0.1>
   }
   swimming {
    angularVelocityScale=0.4
    fictitiousAngularScale=0
    jointName=root
    dragCoefficient=<0 0.5>
    gravityOverride=<0 0 -400>
    liftCoefficient=<0 0.1>
    linearVelocityScale=<0.4 0.2 0.2>
   }
   tactical_Sprint {
    angularVelocityScale=0.1
    fictitiousAngularScale=0.1
    jointName=root
    dragCoefficient=<1.5 1.5>
    gravityOverride=<0 0 -980>
    liftCoefficient=<0.4 0.4>
    linearVelocityScale=<0.2 0.2 0.2>
   }
   ziplining {
    angularVelocityScale=0.3
    fictitiousAngularScale=0.3
    jointName=root
    dragCoefficient=<1 2>
    gravityOverride=<0 0 -980>
    liftCoefficient=<1 2>
    linearVelocityScale=<0 0 0>
   }
  }
 }
 CrouchingAndNotMovingAndNotShape {type=boolean}
 DBNO_Alpha {type=number}
 DeltaTimeLocal {type=number}
 DeltaTimeLocal_0 {type=number}
 DivingAndCape {type=boolean}
 DivingOrDBNO {type=boolean}
 Gravity Override Cape {type=Vector}
 Gravity Override Cape_0 {type=Vector}
 Gravity Override Params {
  type=GravityOverrideParamsStruct
  default {
   front_End {gravityOverride=<0 10 -980>}
   riding_Moving_Targeting {gravityOverride=<0 -200 -800>}
   riding_Standing {gravityOverride=<0 -350 -400>}
   skydive_DivingDown {gravityOverride=<0 0 50>, windAmplitude=<0 0 0>, windFrequency=<0 0 0>}
   skydive_FloatingDown {jointName=pelvis, gravityOverride=<0 0 50>, windAmplitude=<0 0 0>, windFrequency=<0 0 0>, windOffset=<0 0 0>}
   skydive_GoingUp {gravityOverride=<0 -600 -150>}
   skydive_Parachuting {gravityOverride=<0 0 -500>, windAmplitude=<0 0 0>, windFrequency=<0 0 0>}
   skydive_Surfing {gravityOverride=<0 -500 -980>, windAmplitude=<0 0 0>, windFrequency=<0 0 0>}
  }
 }
 Gravity Override Params_0 {
  type=GravityOverrideParamsStruct
  default {
   riding_Moving_Targeting {gravityOverride=<0 -200 -800>}
   riding_Standing {gravityOverride=<0 -350 -400>}
   skydive_DivingDown {gravityOverride=<0 0 50>, windAmplitude=<0 0 0>, windFrequency=<0 0 0>}
   skydive_FloatingDown {jointName=pelvis, gravityOverride=<0 0 50>, windAmplitude=<0 0 0>, windFrequency=<0 0 0>, windOffset=<0 0 0>}
   skydive_GoingUp {gravityOverride=<0 -600 -150>}
   skydive_Parachuting {gravityOverride=<0 -500 -800>, windAmplitude=<0 0 0>, windFrequency=<0 0 0>}
   skydive_Surfing {gravityOverride=<0 -500 -980>, windAmplitude=<0 0 0>, windFrequency=<0 0 0>}
  }
 }
 Gust Local Settings {type=WindGustLocalSettingStruct, default {elapsedTime=0, finalGustStrengthLocal=0, gustDuration=0, gustStrength=0, timeBetweenGusts=0}}
 is sliding {type=number}
 is_WearingCape {type=boolean}
 isFrontEndAndNotEmote {type=boolean}
 isWearingPackBool {type=boolean}
 MovingAnd NotCrouchNotDivingNotEmote {type=boolean}
 Not DiveFloatingAndDiveMode {type=boolean}
 SkelMesh {type=SkeletalMeshComponent}
 thigh_shape {type=boolean}
 thigh_shape_fn_l {type=number}
 thigh_shape_fn_r {type=number}
 thigh_shape_fn_r_0 {type=number}
 thigh_shape_fn_r_1 {type=number}
 Wind Gust Params {
  type=WindGustParamsStruct
  default {
   dBNO {
    maxDownInterpSpeed=5
    maxGustDuration=2
    maxGustStrength=7
    maxTimeBetweenGusts=7
    maxUpInterpSpeed=5
    minDownInterpSpeed=2
    minGustDuration=0.5
    minGustStrength=3
    minTimeBetweenGusts=0
    minUpInterpSpeed=2
    windTurbulenceScalarA=0.1
    windTurbulenceScalarB=1
    windAmplitude=<0 0 0>
    windFrequency=<0 0 0>
    windOffset=<0 0 0>
   }
   dBNO_Carried {
    maxDownInterpSpeed=5
    maxGustDuration=2
    maxGustStrength=7
    maxTimeBetweenGusts=7
    maxUpInterpSpeed=5
    minDownInterpSpeed=2
    minGustDuration=0.5
    minGustStrength=3
    minTimeBetweenGusts=0
    minUpInterpSpeed=2
    windTurbulenceScalarA=0.1
    windTurbulenceScalarB=1
    windAmplitude=<0 0 0>
    windFrequency=<0 0 0>
    windOffset=<0 0 0>
   }
   dBNO_Impulse {
    maxDownInterpSpeed=5
    maxGustDuration=2
    maxGustStrength=7
    maxTimeBetweenGusts=7
    maxUpInterpSpeed=5
    minDownInterpSpeed=2
    minGustDuration=0.5
    minGustStrength=3
    minTimeBetweenGusts=0
    minUpInterpSpeed=2
    windTurbulenceScalarA=0.1
    windTurbulenceScalarB=1
    windAmplitude=<0 0 0>
    windFrequency=<0 0 0>
    windOffset=<0 0 0>
   }
   dBNO_RollBwd {
    maxDownInterpSpeed=5
    maxGustDuration=2
    maxGustStrength=7
    maxTimeBetweenGusts=7
    maxUpInterpSpeed=5
    minDownInterpSpeed=2
    minGustDuration=0.5
    minGustStrength=3
    minTimeBetweenGusts=0
    minUpInterpSpeed=2
    windTurbulenceScalarA=0.1
    windTurbulenceScalarB=1
    windAmplitude=<0 0 0>
    windFrequency=<0 0 0>
    windOffset=<0 0 0>
   }
   dBNO_RollFwd {
    maxDownInterpSpeed=5
    maxGustDuration=2
    maxGustStrength=7
    maxTimeBetweenGusts=7
    maxUpInterpSpeed=5
    minDownInterpSpeed=2
    minGustDuration=0.5
    minGustStrength=3
    minTimeBetweenGusts=0
    minUpInterpSpeed=2
    windTurbulenceScalarA=0.1
    windTurbulenceScalarB=1
    windAmplitude=<0 0 0>
    windFrequency=<0 0 0>
    windOffset=<0 0 0>
   }
   dBNO_Tumbling {
    maxDownInterpSpeed=5
    maxGustDuration=2
    maxGustStrength=7
    maxTimeBetweenGusts=7
    maxUpInterpSpeed=5
    minDownInterpSpeed=2
    minGustDuration=0.5
    minGustStrength=3
    minTimeBetweenGusts=0
    minUpInterpSpeed=2
    windTurbulenceScalarA=0.1
    windTurbulenceScalarB=1
    windAmplitude=<0 0 0>
    windFrequency=<0 0 0>
    windOffset=<0 0 0>
   }
   emoteOrMelee {
    maxDownInterpSpeed=5
    maxGustDuration=2
    maxGustStrength=7
    maxTimeBetweenGusts=7
    maxUpInterpSpeed=5
    minDownInterpSpeed=2
    minGustDuration=0.5
    minGustStrength=3
    minTimeBetweenGusts=0
    minUpInterpSpeed=2
    windTurbulenceScalarA=0.1
    windTurbulenceScalarB=1
    windAmplitude=<30 25 50>
    windFrequency=<2 3 5>
    windOffset=<-50 -10 0>
   }
   front_End {
    maxDownInterpSpeed=5
    maxGustDuration=8
    maxGustStrength=7
    maxTimeBetweenGusts=8
    maxUpInterpSpeed=5
    minDownInterpSpeed=2
    minGustDuration=0.2
    minGustStrength=2
    minTimeBetweenGusts=1
    minUpInterpSpeed=2
    windTurbulenceScalarA=0.1
    windTurbulenceScalarB=1
    windAmplitude=<0 0 0>
    windFrequency=<1.5 2.3 3.7>
    windOffset=<0 0 0>
   }
   onGround_Crouch_Moving {
    maxDownInterpSpeed=5
    maxGustDuration=2
    maxGustStrength=7
    maxTimeBetweenGusts=7
    maxUpInterpSpeed=5
    minDownInterpSpeed=2
    minGustDuration=0.5
    minGustStrength=3
    minTimeBetweenGusts=0
    minUpInterpSpeed=2
    windTurbulenceScalarA=0.1
    windTurbulenceScalarB=1
    windAmplitude=<100 25 75>
    windFrequency=<1.5 0.5 7.5>
    windOffset=<100 -100 -175>
   }
   onGround_Falling {
    maxDownInterpSpeed=5
    maxGustDuration=2
    maxGustStrength=7
    maxTimeBetweenGusts=7
    maxUpInterpSpeed=5
    minDownInterpSpeed=2
    minGustDuration=0.5
    minGustStrength=3
    minTimeBetweenGusts=0
    minUpInterpSpeed=2
    windTurbulenceScalarA=0.1
    windTurbulenceScalarB=1
    windAmplitude=<0 0 0>
    windFrequency=<0 0 0>
    windOffset=<0 0 0>
   }
   onGround_Moving {
    maxDownInterpSpeed=5
    maxGustDuration=2
    maxGustStrength=7
    maxTimeBetweenGusts=7
    maxUpInterpSpeed=5
    minDownInterpSpeed=2
    minGustDuration=0.5
    minGustStrength=3
    minTimeBetweenGusts=0
    minUpInterpSpeed=2
    windTurbulenceScalarA=0.1
    windTurbulenceScalarB=1
    windAmplitude=<0 0 0>
    windFrequency=<0 0 0>
    windOffset=<0 0 0>
   }
   onGround_Standing {
    maxDownInterpSpeed=5
    maxGustDuration=8
    maxGustStrength=7
    maxTimeBetweenGusts=8
    maxUpInterpSpeed=5
    minDownInterpSpeed=2
    minGustDuration=0.2
    minGustStrength=2
    minTimeBetweenGusts=1
    minUpInterpSpeed=2
    windTurbulenceScalarA=0.1
    windTurbulenceScalarB=1
    windAmplitude=<15 20 24>
    windFrequency=<1.5 2.3 3.7>
    windOffset=<-100 0 -70>
   }
   rail_Grinding {
    maxDownInterpSpeed=5
    maxGustDuration=8
    maxGustStrength=7
    maxTimeBetweenGusts=8
    maxUpInterpSpeed=5
    minDownInterpSpeed=2
    minGustDuration=0.2
    minGustStrength=3
    minTimeBetweenGusts=1
    minUpInterpSpeed=2
    windTurbulenceScalarA=0.1
    windTurbulenceScalarB=0.8
    windAmplitude=<20 65 42>
    windFrequency=<5 6 8.6>
    windOffset=<50 -500 -400>
   }
   riding_Moving {
    maxDownInterpSpeed=5
    maxGustDuration=2
    maxGustStrength=7
    maxTimeBetweenGusts=7
    maxUpInterpSpeed=5
    minDownInterpSpeed=2
    minGustDuration=0.5
    minGustStrength=3
    minTimeBetweenGusts=0
    minUpInterpSpeed=2
    windTurbulenceScalarA=0.1
    windTurbulenceScalarB=1
    windAmplitude=<0 0 0>
    windFrequency=<0 0 0>
    windOffset=<0 0 0>
   }
   riding_Moving_Targeting {
    maxDownInterpSpeed=5
    maxGustDuration=2
    maxGustStrength=7
    maxTimeBetweenGusts=7
    maxUpInterpSpeed=5
    minDownInterpSpeed=2
    minGustDuration=0.5
    minGustStrength=3
    minTimeBetweenGusts=0
    minUpInterpSpeed=2
    windTurbulenceScalarA=0.1
    windTurbulenceScalarB=1
    windAmplitude=<0 0 0>
    windFrequency=<0 0 0>
    windOffset=<0 0 0>
   }
   riding_Standing {
    maxDownInterpSpeed=5
    maxGustDuration=2
    maxGustStrength=7
    maxTimeBetweenGusts=7
    maxUpInterpSpeed=5
    minDownInterpSpeed=2
    minGustDuration=0.5
    minGustStrength=3
    minTimeBetweenGusts=0
    minUpInterpSpeed=2
    windTurbulenceScalarA=0.1
    windTurbulenceScalarB=1
    windAmplitude=<0 0 0>
    windFrequency=<0 0 0>
    windOffset=<0 0 0>
   }
   skydive_DivingDown {
    maxDownInterpSpeed=5
    maxGustDuration=8
    maxGustStrength=7
    maxTimeBetweenGusts=8
    maxUpInterpSpeed=5
    minDownInterpSpeed=2
    minGustDuration=0.2
    minGustStrength=3
    minTimeBetweenGusts=1
    minUpInterpSpeed=2
    windTurbulenceScalarA=0.1
    windTurbulenceScalarB=4
    windAmplitude=<20 25 150>
    windFrequency=<2 4 30>
    windOffset=<-150 -150 0>
   }
   skydive_FloatingDown {
    maxDownInterpSpeed=5
    maxGustDuration=8
    maxGustStrength=1
    maxTimeBetweenGusts=8
    maxUpInterpSpeed=5
    minDownInterpSpeed=2
    minGustDuration=0.2
    minGustStrength=0
    minTimeBetweenGusts=1
    minUpInterpSpeed=2
    windTurbulenceScalarA=0.5
    windTurbulenceScalarB=1.5
    windAmplitude=<70 20 50>
    windFrequency=<10 2 3>
    windOffset=<0 -300 100>
   }
   skydive_GoingUp {
    maxDownInterpSpeed=5
    maxGustDuration=8
    maxGustStrength=7
    maxTimeBetweenGusts=8
    maxUpInterpSpeed=5
    minDownInterpSpeed=2
    minGustDuration=0.2
    minGustStrength=3
    minTimeBetweenGusts=1
    minUpInterpSpeed=2
    windTurbulenceScalarA=0.1
    windTurbulenceScalarB=0.8
    windAmplitude=<50 20 35>
    windFrequency=<8 2 3>
    windOffset=<0 -500 25>
   }
   skydive_Parachuting {
    maxDownInterpSpeed=5
    maxGustDuration=8
    maxGustStrength=1
    maxTimeBetweenGusts=8
    maxUpInterpSpeed=5
    minDownInterpSpeed=2
    minGustDuration=5
    minGustStrength=0
    minTimeBetweenGusts=5
    minUpInterpSpeed=2
    windTurbulenceScalarA=0.2
    windTurbulenceScalarB=1
    windAmplitude=<50 50 75>
    windFrequency=<15 6 2>
    windOffset=<0 -250 -200>
   }
   skydive_Surfing {
    maxDownInterpSpeed=5
    maxGustDuration=8
    maxGustStrength=1
    maxTimeBetweenGusts=8
    maxUpInterpSpeed=5
    minDownInterpSpeed=2
    minGustDuration=5
    minGustStrength=0
    minTimeBetweenGusts=5
    minUpInterpSpeed=2
    windTurbulenceScalarA=0.2
    windTurbulenceScalarB=1
    windAmplitude=<50 50 75>
    windFrequency=<15 6 2>
    windOffset=<0 -250 -200>
   }
   swimming {
    maxDownInterpSpeed=5
    maxGustDuration=2
    maxGustStrength=7
    maxTimeBetweenGusts=7
    maxUpInterpSpeed=5
    minDownInterpSpeed=2
    minGustDuration=0.5
    minGustStrength=3
    minTimeBetweenGusts=0
    minUpInterpSpeed=2
    windTurbulenceScalarA=0.1
    windTurbulenceScalarB=1
    windAmplitude=<10 25 20>
    windFrequency=<1.5 0.5 7.5>
    windOffset=<0 -150 50>
   }
   tactical_Sprint {
    maxDownInterpSpeed=5
    maxGustDuration=2
    maxGustStrength=7
    maxTimeBetweenGusts=7
    maxUpInterpSpeed=5
    minDownInterpSpeed=2
    minGustDuration=0.5
    minGustStrength=3
    minTimeBetweenGusts=0
    minUpInterpSpeed=2
    windTurbulenceScalarA=0.1
    windTurbulenceScalarB=1
    windAmplitude=<0 0 0>
    windFrequency=<0 0 0>
    windOffset=<0 0 0>
   }
   ziplining {
    maxDownInterpSpeed=5
    maxGustDuration=8
    maxGustStrength=7
    maxTimeBetweenGusts=8
    maxUpInterpSpeed=5
    minDownInterpSpeed=2
    minGustDuration=0.2
    minGustStrength=3
    minTimeBetweenGusts=1
    minUpInterpSpeed=2
    windTurbulenceScalarA=0.1
    windTurbulenceScalarB=4
    windAmplitude=<0 25 35>
    windFrequency=<0 4 10>
    windOffset=<0 -70 25>
   }
  }
 }
]
