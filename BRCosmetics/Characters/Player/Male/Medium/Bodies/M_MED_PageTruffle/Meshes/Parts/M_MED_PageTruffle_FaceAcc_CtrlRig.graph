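# Control Rig for the PageTruffle face accessory mesh that adjusts neck and head bones as the head moves.
# It first loops over neck_01, neck_02 and head, damping their local Y and Z Euler rotation.
# Two spherical pose readers then track head orientation relative to the pelvis, one weighting a corrective offset on neck_fix, the other a corrective rotation on head.
# The hierarchy carries facial expression curves, which the graph never uses.
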ControlRigBlueprint /BRCosmetics/Characters/Player/Male/Medium/Bodies/M_MED_PageTruffle/Meshes/Parts/M_MED_PageTruffle_FaceAcc_CtrlRig
graphs [
 ContainedGraph {nodes [
   n1 = RigVMFunctionEntryNode()
   SequenceAggregate = &RigVMFunction_Sequence(ExecuteContext=@n1.ExecuteContext)
   SequenceAggregate_1 = &RigVMFunction_Sequence(ExecuteContext=@SequenceAggregate.B)
   SequenceAggregate_2 = &RigVMFunction_Sequence(ExecuteContext=@SequenceAggregate_1.B)
   RigVMFunctionReturnNode(A=@SequenceAggregate.A, B=@SequenceAggregate_1.A, C=@SequenceAggregate_2.A, D=@SequenceAggregate_2.B)
  ]}
 RigVMFunctionLibrary {links [], nodes []}
 RigVMModel {nodes [
   RigUnit_BeginExecution = &RigUnit_BeginExecution()
   SequenceAggregate = RigVMAggregateNode(ExecuteContext=@RigUnit_BeginExecution.ExecutePin)
   DISPATCH_RigVMDispatch_ArrayIterator = RigVMDispatchNode(
    Array=&RigUnit_ItemArray(Items=((Type=Bone,Name="neck_01"),(Type=Bone,Name="neck_02"),(Type=Bone,Name="head"))).Items
    ExecuteContext=@SequenceAggregate.A
   ) {template="DISPATCH_RigVMDispatch_ArrayIterator(in Array,out Element,out Index,out Count,out Ratio)"}
   RigUnit_GetTransform = &RigUnit_GetTransform(Space=LocalSpace, bInitial=false, Item=@DISPATCH_RigVMDispatch_ArrayIterator.Element)
   RigVMFunction_MathQuaternionToEuler = &RigVMFunction_MathQuaternionToEuler(RotationOrder=ZXY, Value=@RigUnit_GetTransform.Transform.Rotation)
   Multiply = &RigVMFunction_MathDoubleMul(B=0.8, A=@RigVMFunction_MathQuaternionToEuler.Result.Z)
   RigVMFunction_MathQuaternionFromEuler = &RigVMFunction_MathQuaternionFromEuler(
    Euler=(X=0,Y=0,Z=0)
    RotationOrder=ZXY
    Euler.Z=@Multiply.Result
    Euler.Y=&RigVMFunction_MathDoubleMul(B=0.9, A=@RigVMFunction_MathQuaternionToEuler.Result.Y).Result
    Euler.X=&RigVMFunction_MathDoubleMul(B=1, A=@RigVMFunction_MathQuaternionToEuler.Result.X).Result
   )
   > RigUnit_SetTransform = &RigUnit_SetTransform(
    Space=LocalSpace
    Value=(Rotation=(X=-0.0029,Y=0.7071,Z=0.0029,W=-0.7071),Translation=(X=0,Y=-1.6873,Z=152.0342),Scale3D=(X=1,Y=1,Z=1))
    Weight=1
    bInitial=False
    Item=@DISPATCH_RigVMDispatch_ArrayIterator.Element
    Value.Rotation=@RigVMFunction_MathQuaternionFromEuler.Result
    Value.Translation=@RigUnit_GetTransform.Transform.Translation
    Value.Scale3D=@RigUnit_GetTransform.Transform.Scale3D
   )
   n1 = &RigUnit_SphericalPoseReader(
    ActiveRegionScaleFactors=(PositiveWidth=2,NegativeWidth=2,PositiveHeight=2,NegativeHeight=2)
    ActiveRegionSize=0.1
    Debug=(bDrawDebug=false,DebugScale=25,DebugSegments=20,DebugThickness=0.25)
    DriverAxis=(X=0,Y=1,Z=0)
    DriverItem=(Type=Bone,Name="head")
    FalloffRegionScaleFactors=(PositiveWidth=1.35,NegativeWidth=1.25,PositiveHeight=1.35,NegativeHeight=1.7)
    FalloffSize=0.5
    FlipHeightScaling=false
    FlipWidthScaling=true
    OptionalParentItem=(Type=Bone,Name="pelvis")
    RotationOffset=(X=210,Y=-5,Z=50)
    ExecutePin=@SequenceAggregate.B
   )
   > &RigUnit_ModifyTransforms(
    ItemToModify=((Item=(Type=Bone,Name="neck_fix"),Transform=(Rotation=(X=0,Y=0,Z=0.2588,W=0.9659),Translation=(X=0,Y=-4,Z=0),Scale3D=(X=1,Y=1,Z=1))))
    Weight=@n1.OutputParam
   )
   > n2 = &RigUnit_SphericalPoseReader(
    ActiveRegionScaleFactors=(PositiveWidth=2,NegativeWidth=2,PositiveHeight=2,NegativeHeight=2)
    ActiveRegionSize=0.1
    Debug=(bDrawDebug=false,DebugScale=25,DebugSegments=20,DebugThickness=0.25)
    DriverAxis=(X=0,Y=1,Z=0)
    DriverItem=(Type=Bone,Name="head")
    FalloffRegionScaleFactors=(PositiveWidth=1.35,NegativeWidth=1.25,PositiveHeight=1.35,NegativeHeight=1.7)
    FalloffSize=0.4
    FlipHeightScaling=false
    FlipWidthScaling=true
    OptionalParentItem=(Type=Bone,Name="pelvis")
    RotationOffset=(X=90,Y=0,Z=85)
   )
   > &RigUnit_ModifyTransforms(
    ItemToModify=((Item=(Type=Bone,Name="head"),Transform=(Rotation=(X=0,Y=0,Z=-0.1478,W=0.989),Translation=(X=0,Y=0,Z=0),Scale3D=(X=1,Y=1,Z=1))))
    Weight=@n2.OutputParam
   )
  ]}
]
hierarchy [
 root {loc=<0 0 0>, rot=<0 0 0 1>}
 pelvis {parent=root, loc=<0 -0.7071 93.1191>, rot=<0.0127 0.707 -0.0127 -0.707>}
 spine_01 {parent=pelvis, loc=<6.7494 -0.6388 0>, rot=<0 0 0.034 -0.9994>}
 spine_02 {parent=spine_01, loc=<7.252 -0.035 0>, rot=<0 0 0.0238 -0.9997>}
 spine_03 {parent=spine_02, loc=<6.0106 0 0>, rot=<0 0 0.0149 -0.9999>}
 spine_04 {parent=spine_03, loc=<8.1024 0 0>, rot=<0 0 -0.0132 -0.9999>}
 spine_05 {parent=spine_04, loc=<10.0142 0 0>, rot=<0 0 -0.0015 -1>}
 neck_01 {parent=spine_05, loc=<11.3138 0.1529 0>, rot=<0 0 -0.2003 -0.9797>}
 neck_02 {parent=neck_01, loc=<4.5627 0 0>, rot=<0 0 0.0504 -0.9987>}
 head {parent=neck_02, loc=<5.3914 0 0>, rot=<0 0 0.115 -0.9934>}
 neck_fix {parent=neck_02, loc=<3.1846 4.5041 0>, rot=<0 0 0.115 -0.9934>}
 clavicle_l {parent=spine_05, loc=<4.3575 -0.1466 -2.3223>, rot=<0.0465 -0.7289 -0.009 -0.6829>}
 upperarm_l {parent=clavicle_l, loc=<15.5111 0 0>, rot=<0.0226 -0.3234 -0.0088 -0.9459>}
 clavicle_r {parent=spine_05, loc=<4.3575 -0.1466 2.3223>, rot=<0.7289 0.0465 -0.6829 0.009>}
 upperarm_r {parent=clavicle_r, loc=<-15.5111 0 0>, rot=<0.0226 -0.3234 -0.0088 -0.9459>}
 CTRL_expressions_browDownL {type=Curve, loc=<0 0 0>, rot=<0 0 0 1>}
 CTRL_expressions_browDownR {type=Curve, loc=<0 0 0>, rot=<0 0 0 1>}
 CTRL_expressions_browLateralL {type=Curve, loc=<0 0 0>, rot=<0 0 0 1>}
 CTRL_expressions_browLateralR {type=Curve, loc=<0 0 0>, rot=<0 0 0 1>}
 CTRL_expressions_browRaiseInL {type=Curve, loc=<0 0 0>, rot=<0 0 0 1>}
 CTRL_expressions_browRaiseInR {type=Curve, loc=<0 0 0>, rot=<0 0 0 1>}
 CTRL_expressions_browRaiseOuterL {type=Curve, loc=<0 0 0>, rot=<0 0 0 1>}
 CTRL_expressions_browRaiseOuterR {type=Curve, loc=<0 0 0>, rot=<0 0 0 1>}
 CTRL_expressions_browUINL {type=Curve, loc=<0 0 0>, rot=<0 0 0 1>}
 CTRL_expressions_browUINR {type=Curve, loc=<0 0 0>, rot=<0 0 0 1>}
 CTRL_expressions_browUOUTL {type=Curve, loc=<0 0 0>, rot=<0 0 0 1>}
 CTRL_expressions_browUOUTR {type=Curve, loc=<0 0 0>, rot=<0 0 0 1>}
 CTRL_expressions_browDINL {type=Curve, loc=<0 0 0>, rot=<0 0 0 1>}
 CTRL_expressions_browDINR {type=Curve, loc=<0 0 0>, rot=<0 0 0 1>}
 CTRL_expressions_browDOUTL {type=Curve, loc=<0 0 0>, rot=<0 0 0 1>}
 CTRL_expressions_browDOUTR {type=Curve, loc=<0 0 0>, rot=<0 0 0 1>}
 CTRL_expressions_eyeBlinkL {type=Curve, loc=<0 0 0>, rot=<0 0 0 1>}
 CTRL_expressions_eyeBlinkR {type=Curve, loc=<0 0 0>, rot=<0 0 0 1>}
 CTRL_expressions_eyeWidenL {type=Curve, loc=<0 0 0>, rot=<0 0 0 1>}
 CTRL_expressions_eyeWidenR {type=Curve, loc=<0 0 0>, rot=<0 0 0 1>}
 CTRL_expressions_eyeSquintInnerL {type=Curve, loc=<0 0 0>, rot=<0 0 0 1>}
 CTRL_expressions_eyeSquintInnerR {type=Curve, loc=<0 0 0>, rot=<0 0 0 1>}
 CTRL_expressions_eyeCheekRaiseL {type=Curve, loc=<0 0 0>, rot=<0 0 0 1>}
 CTRL_expressions_eyeCheekRaiseR {type=Curve, loc=<0 0 0>, rot=<0 0 0 1>}
 CTRL_expressions_eyeUpperLidUpL {type=Curve, loc=<0 0 0>, rot=<0 0 0 1>}
 CTRL_expressions_eyeUpperLidUpR {type=Curve, loc=<0 0 0>, rot=<0 0 0 1>}
 CTRL_expressions_eyeRelaxL {type=Curve, loc=<0 0 0>, rot=<0 0 0 1>}
 CTRL_expressions_eyeRelaxR {type=Curve, loc=<0 0 0>, rot=<0 0 0 1>}
 CTRL_expressions_eyeLowerLidUpL {type=Curve, loc=<0 0 0>, rot=<0 0 0 1>}
 CTRL_expressions_eyeLowerLidUpR {type=Curve, loc=<0 0 0>, rot=<0 0 0 1>}
 CTRL_expressions_eyeLowerLidDownL {type=Curve, loc=<0 0 0>, rot=<0 0 0 1>}
 CTRL_expressions_eyeLowerLidDownR {type=Curve, loc=<0 0 0>, rot=<0 0 0 1>}
 CTRL_expressions_eyelidRollPosUL {type=Curve, loc=<0 0 0>, rot=<0 0 0 1>}
 CTRL_expressions_eyelidRollPosUR {type=Curve, loc=<0 0 0>, rot=<0 0 0 1>}
 CTRL_expressions_eyelidRollPosDL {type=Curve, loc=<0 0 0>, rot=<0 0 0 1>}
 CTRL_expressions_eyelidRollPosDR {type=Curve, loc=<0 0 0>, rot=<0 0 0 1>}
 CTRL_expressions_eyelidRollNegUL {type=Curve, loc=<0 0 0>, rot=<0 0 0 1>}
 CTRL_expressions_eyelidRollNegUR {type=Curve, loc=<0 0 0>, rot=<0 0 0 1>}
 CTRL_expressions_eyelidRollNegDL {type=Curve, loc=<0 0 0>, rot=<0 0 0 1>}
 CTRL_expressions_eyelidRollNegDR {type=Curve, loc=<0 0 0>, rot=<0 0 0 1>}
 CTRL_expressions_eyeLookUpL {type=Curve, loc=<0 0 0>, rot=<0 0 0 1>}
 CTRL_expressions_eyeLookUpR {type=Curve, loc=<0 0 0>, rot=<0 0 0 1>}
 CTRL_expressions_eyeLookDownL {type=Curve, loc=<0 0 0>, rot=<0 0 0 1>}
 CTRL_expressions_eyeLookDownR {type=Curve, loc=<0 0 0>, rot=<0 0 0 1>}
 CTRL_expressions_eyeLookLeftL {type=Curve, loc=<0 0 0>, rot=<0 0 0 1>}
 CTRL_expressions_eyeLookLeftR {type=Curve, loc=<0 0 0>, rot=<0 0 0 1>}
 CTRL_expressions_eyeLookRightL {type=Curve, loc=<0 0 0>, rot=<0 0 0 1>}
 CTRL_expressions_eyeLookRightR {type=Curve, loc=<0 0 0>, rot=<0 0 0 1>}
 CTRL_expressions_eyeEyelidsReduce {type=Curve, loc=<0 0 0>, rot=<0 0 0 1>}
 CTRL_expressions_noseWrinkleL {type=Curve, loc=<0 0 0>, rot=<0 0 0 1>}
 CTRL_expressions_noseWrinkleR {type=Curve, loc=<0 0 0>, rot=<0 0 0 1>}
 CTRL_expressions_noseWrinkleUpperL {type=Curve, loc=<0 0 0>, rot=<0 0 0 1>}
 CTRL_expressions_noseWrinkleUpperR {type=Curve, loc=<0 0 0>, rot=<0 0 0 1>}
 CTRL_expressions_noseNostrilDepressL {type=Curve, loc=<0 0 0>, rot=<0 0 0 1>}
 CTRL_expressions_noseNostrilDepressR {type=Curve, loc=<0 0 0>, rot=<0 0 0 1>}
 CTRL_expressions_noseNostrilDilateL {type=Curve, loc=<0 0 0>, rot=<0 0 0 1>}
 CTRL_expressions_noseNostrilDilateR {type=Curve, loc=<0 0 0>, rot=<0 0 0 1>}
 CTRL_expressions_noseNostrilCompressL {type=Curve, loc=<0 0 0>, rot=<0 0 0 1>}
 CTRL_expressions_noseNostrilCompressR {type=Curve, loc=<0 0 0>, rot=<0 0 0 1>}
 CTRL_expressions_noseRotateL {type=Curve, loc=<0 0 0>, rot=<0 0 0 1>}
 CTRL_expressions_noseRotateR {type=Curve, loc=<0 0 0>, rot=<0 0 0 1>}
 CTRL_expressions_noseSquash {type=Curve, loc=<0 0 0>, rot=<0 0 0 1>}
 CTRL_expressions_noseStretch {type=Curve, loc=<0 0 0>, rot=<0 0 0 1>}
 CTRL_expressions_mouthCheekSuckL {type=Curve, loc=<0 0 0>, rot=<0 0 0 1>}
 CTRL_expressions_mouthCheekSuckR {type=Curve, loc=<0 0 0>, rot=<0 0 0 1>}
 CTRL_expressions_mouthCheekBlowL {type=Curve, loc=<0 0 0>, rot=<0 0 0 1>}
 CTRL_expressions_mouthCheekBlowR {type=Curve, loc=<0 0 0>, rot=<0 0 0 1>}
 CTRL_expressions_mouthLipsBlowL {type=Curve, loc=<0 0 0>, rot=<0 0 0 1>}
 CTRL_expressions_mouthLipsBlowR {type=Curve, loc=<0 0 0>, rot=<0 0 0 1>}
 CTRL_expressions_mouthLeft {type=Curve, loc=<0 0 0>, rot=<0 0 0 1>}
 CTRL_expressions_mouthRight {type=Curve, loc=<0 0 0>, rot=<0 0 0 1>}
 CTRL_expressions_mouthUp {type=Curve, loc=<0 0 0>, rot=<0 0 0 1>}
 CTRL_expressions_mouthDown {type=Curve, loc=<0 0 0>, rot=<0 0 0 1>}
 CTRL_expressions_mouthUpperLipRaiseL {type=Curve, loc=<0 0 0>, rot=<0 0 0 1>}
 CTRL_expressions_mouthUpperLipRaiseR {type=Curve, loc=<0 0 0>, rot=<0 0 0 1>}
 CTRL_expressions_mouthLowerLipDepressL {type=Curve, loc=<0 0 0>, rot=<0 0 0 1>}
 CTRL_expressions_mouthLowerLipDepressR {type=Curve, loc=<0 0 0>, rot=<0 0 0 1>}
 CTRL_expressions_mouthCornerPullL {type=Curve, loc=<0 0 0>, rot=<0 0 0 1>}
 CTRL_expressions_mouthCornerPullR {type=Curve, loc=<0 0 0>, rot=<0 0 0 1>}
 CTRL_expressions_mouthStretchL {type=Curve, loc=<0 0 0>, rot=<0 0 0 1>}
 CTRL_expressions_mouthStretchR {type=Curve, loc=<0 0 0>, rot=<0 0 0 1>}
 CTRL_expressions_mouthDimpleL {type=Curve, loc=<0 0 0>, rot=<0 0 0 1>}
 CTRL_expressions_mouthDimpleR {type=Curve, loc=<0 0 0>, rot=<0 0 0 1>}
 CTRL_expressions_mouthCornerDepressL {type=Curve, loc=<0 0 0>, rot=<0 0 0 1>}
 CTRL_expressions_mouthCornerDepressR {type=Curve, loc=<0 0 0>, rot=<0 0 0 1>}
 CTRL_expressions_mouthPressUL {type=Curve, loc=<0 0 0>, rot=<0 0 0 1>}
 CTRL_expressions_mouthPressUR {type=Curve, loc=<0 0 0>, rot=<0 0 0 1>}
 CTRL_expressions_mouthPressDL {type=Curve, loc=<0 0 0>, rot=<0 0 0 1>}
 CTRL_expressions_mouthPressDR {type=Curve, loc=<0 0 0>, rot=<0 0 0 1>}
 CTRL_expressions_mouthLipsPurseUL {type=Curve, loc=<0 0 0>, rot=<0 0 0 1>}
 CTRL_expressions_mouthLipsPurseUR {type=Curve, loc=<0 0 0>, rot=<0 0 0 1>}
 CTRL_expressions_mouthLipsPurseDL {type=Curve, loc=<0 0 0>, rot=<0 0 0 1>}
 CTRL_expressions_mouthLipsPurseDR {type=Curve, loc=<0 0 0>, rot=<0 0 0 1>}
 CTRL_expressions_mouthLipsTowardsUL {type=Curve, loc=<0 0 0>, rot=<0 0 0 1>}
 CTRL_expressions_mouthLipsTowardsUR {type=Curve, loc=<0 0 0>, rot=<0 0 0 1>}
 CTRL_expressions_mouthLipsTowardsDL {type=Curve, loc=<0 0 0>, rot=<0 0 0 1>}
 CTRL_expressions_mouthLipsTowardsDR {type=Curve, loc=<0 0 0>, rot=<0 0 0 1>}
 CTRL_expressions_mouthFunnelUL {type=Curve, loc=<0 0 0>, rot=<0 0 0 1>}
 CTRL_expressions_mouthFunnelUR {type=Curve, loc=<0 0 0>, rot=<0 0 0 1>}
 CTRL_expressions_mouthFunnelDL {type=Curve, loc=<0 0 0>, rot=<0 0 0 1>}
 CTRL_expressions_mouthFunnelDR {type=Curve, loc=<0 0 0>, rot=<0 0 0 1>}
 CTRL_expressions_mouthLipsTogetherUL {type=Curve, loc=<0 0 0>, rot=<0 0 0 1>}
 CTRL_expressions_mouthLipsTogetherUR {type=Curve, loc=<0 0 0>, rot=<0 0 0 1>}
 CTRL_expressions_mouthLipsTogetherDL {type=Curve, loc=<0 0 0>, rot=<0 0 0 1>}
 CTRL_expressions_mouthLipsTogetherDR {type=Curve, loc=<0 0 0>, rot=<0 0 0 1>}
 CTRL_expressions_mouthUpperLipBiteL {type=Curve, loc=<0 0 0>, rot=<0 0 0 1>}
 CTRL_expressions_mouthUpperLipBiteR {type=Curve, loc=<0 0 0>, rot=<0 0 0 1>}
 CTRL_expressions_mouthLowerLipBiteL {type=Curve, loc=<0 0 0>, rot=<0 0 0 1>}
 CTRL_expressions_mouthLowerLipBiteR {type=Curve, loc=<0 0 0>, rot=<0 0 0 1>}
 CTRL_expressions_mouthCornerSharpenUL {type=Curve, loc=<0 0 0>, rot=<0 0 0 1>}
 CTRL_expressions_mouthCornerSharpenUR {type=Curve, loc=<0 0 0>, rot=<0 0 0 1>}
 CTRL_expressions_mouthCornerSharpenDL {type=Curve, loc=<0 0 0>, rot=<0 0 0 1>}
 CTRL_expressions_mouthCornerSharpenDR {type=Curve, loc=<0 0 0>, rot=<0 0 0 1>}
 CTRL_expressions_mouthCornerRounderUL {type=Curve, loc=<0 0 0>, rot=<0 0 0 1>}
 CTRL_expressions_mouthCornerRounderUR {type=Curve, loc=<0 0 0>, rot=<0 0 0 1>}
 CTRL_expressions_mouthCornerRounderDL {type=Curve, loc=<0 0 0>, rot=<0 0 0 1>}
 CTRL_expressions_mouthCornerRounderDR {type=Curve, loc=<0 0 0>, rot=<0 0 0 1>}
 CTRL_expressions_mouthUpperLipTowardsTeethL {type=Curve, loc=<0 0 0>, rot=<0 0 0 1>}
 CTRL_expressions_mouthUpperLipTowardsTeethR {type=Curve, loc=<0 0 0>, rot=<0 0 0 1>}
 CTRL_expressions_mouthLowerLipTowardsTeethL {type=Curve, loc=<0 0 0>, rot=<0 0 0 1>}
 CTRL_expressions_mouthLowerLipTowardsTeethR {type=Curve, loc=<0 0 0>, rot=<0 0 0 1>}
 CTRL_expressions_mouthCornerUpL {type=Curve, loc=<0 0 0>, rot=<0 0 0 1>}
 CTRL_expressions_mouthCornerUpR {type=Curve, loc=<0 0 0>, rot=<0 0 0 1>}
 CTRL_expressions_mouthCornerDownL {type=Curve, loc=<0 0 0>, rot=<0 0 0 1>}
 CTRL_expressions_mouthCornerDownR {type=Curve, loc=<0 0 0>, rot=<0 0 0 1>}
 CTRL_expressions_mouthCornerWideL {type=Curve, loc=<0 0 0>, rot=<0 0 0 1>}
 CTRL_expressions_mouthCornerWideR {type=Curve, loc=<0 0 0>, rot=<0 0 0 1>}
 CTRL_expressions_mouthCornerNarrowL {type=Curve, loc=<0 0 0>, rot=<0 0 0 1>}
 CTRL_expressions_mouthCornerNarrowR {type=Curve, loc=<0 0 0>, rot=<0 0 0 1>}
 CTRL_expressions_tongueUp {type=Curve, loc=<0 0 0>, rot=<0 0 0 1>}
 CTRL_expressions_tongueDown {type=Curve, loc=<0 0 0>, rot=<0 0 0 1>}
 CTRL_expressions_tongueLeft {type=Curve, loc=<0 0 0>, rot=<0 0 0 1>}
 CTRL_expressions_tongueRight {type=Curve, loc=<0 0 0>, rot=<0 0 0 1>}
 CTRL_expressions_tongueOut {type=Curve, loc=<0 0 0>, rot=<0 0 0 1>}
 CTRL_expressions_tongueIn {type=Curve, loc=<0 0 0>, rot=<0 0 0 1>}
 CTRL_expressions_tongueRollUp {type=Curve, loc=<0 0 0>, rot=<0 0 0 1>}
 CTRL_expressions_tongueRollDown {type=Curve, loc=<0 0 0>, rot=<0 0 0 1>}
 CTRL_expressions_tongueRollLeft {type=Curve, loc=<0 0 0>, rot=<0 0 0 1>}
 CTRL_expressions_tongueRollRight {type=Curve, loc=<0 0 0>, rot=<0 0 0 1>}
 CTRL_expressions_tongueWide {type=Curve, loc=<0 0 0>, rot=<0 0 0 1>}
 CTRL_expressions_tongueNarrow {type=Curve, loc=<0 0 0>, rot=<0 0 0 1>}
 CTRL_expressions_jawOpen {type=Curve, loc=<0 0 0>, rot=<0 0 0 1>}
 CTRL_expressions_jawLeft {type=Curve, loc=<0 0 0>, rot=<0 0 0 1>}
 CTRL_expressions_jawRight {type=Curve, loc=<0 0 0>, rot=<0 0 0 1>}
 CTRL_expressions_jawFwd {type=Curve, loc=<0 0 0>, rot=<0 0 0 1>}
 CTRL_expressions_jawBack {type=Curve, loc=<0 0 0>, rot=<0 0 0 1>}
 CTRL_expressions_jawChinRaiseDL {type=Curve, loc=<0 0 0>, rot=<0 0 0 1>}
 CTRL_expressions_jawChinRaiseDR {type=Curve, loc=<0 0 0>, rot=<0 0 0 1>}
 CTRL_expressions_jawChinRaiseUL {type=Curve, loc=<0 0 0>, rot=<0 0 0 1>}
 CTRL_expressions_jawChinRaiseUR {type=Curve, loc=<0 0 0>, rot=<0 0 0 1>}
 CTRL_expressions_jawOpenExtreme {type=Curve, loc=<0 0 0>, rot=<0 0 0 1>}
 CTRL_expressions_teethUpU {type=Curve, loc=<0 0 0>, rot=<0 0 0 1>}
 CTRL_expressions_teethUpD {type=Curve, loc=<0 0 0>, rot=<0 0 0 1>}
 CTRL_expressions_teethDownU {type=Curve, loc=<0 0 0>, rot=<0 0 0 1>}
 CTRL_expressions_teethDownD {type=Curve, loc=<0 0 0>, rot=<0 0 0 1>}
 CTRL_expressions_teethLeftU {type=Curve, loc=<0 0 0>, rot=<0 0 0 1>}
 CTRL_expressions_teethLeftD {type=Curve, loc=<0 0 0>, rot=<0 0 0 1>}
 CTRL_expressions_teethRightU {type=Curve, loc=<0 0 0>, rot=<0 0 0 1>}
 CTRL_expressions_teethRightD {type=Curve, loc=<0 0 0>, rot=<0 0 0 1>}
 CTRL_expressions_teethFwdU {type=Curve, loc=<0 0 0>, rot=<0 0 0 1>}
 CTRL_expressions_teethFwdD {type=Curve, loc=<0 0 0>, rot=<0 0 0 1>}
 CTRL_expressions_teethBackU {type=Curve, loc=<0 0 0>, rot=<0 0 0 1>}
 CTRL_expressions_teethBackD {type=Curve, loc=<0 0 0>, rot=<0 0 0 1>}
 CTRL_expressions_teethRotateFwdU {type=Curve, loc=<0 0 0>, rot=<0 0 0 1>}
 CTRL_expressions_teethRotateFwdD {type=Curve, loc=<0 0 0>, rot=<0 0 0 1>}
 CTRL_expressions_teethRotateBackU {type=Curve, loc=<0 0 0>, rot=<0 0 0 1>}
 CTRL_expressions_teethRotateBackD {type=Curve, loc=<0 0 0>, rot=<0 0 0 1>}
 DisableFaceOverride {type=Curve, loc=<0 0 0>, rot=<0 0 0 1>}
 is_3L {type=Curve, loc=<0 0 0>, rot=<0 0 0 1>}
 is_face_bone_anim {type=Curve, loc=<0 0 0>, rot=<0 0 0 1>}
]
